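# Unlit masked explosion material for Niagara mesh particles, in the prototype FX of the Hawaii stamina item.
# Panning noise, weighted by vertex red and offset by Erode, forms an erosion value.
# A contrast-adjusted Fresnel, masked along inverted V, blends center and dark colors, tinted by particle color for emissive.
# Dynamic parameters drive erosion and two variants of vertex-red-weighted displacement along normals, one with adjustable contrast.
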Material /StaminaGameplay/Gameplay/Items/Hawaii/PROTOTYPE/FX/M_StaminaHawaii_Explosion2
graph {
 properties {
  bUsedWithNiagaraMeshParticles=true
  blendMode=BLEND_Masked
  pixelDepthOffsetMode=PDOM_Legacy
  shadingModel=MSM_Unlit
  usedShadingModels=Unlit
 }
 nodes [
  n1 = TextureCoordinate()
  n2 = Multiply(A=ComponentMask(None=@n1) {r=true}, B=ScalarParameter() {defaultValue=2.5, group=Distortion UV, parameterName=Mask Tile U})
  n3 = Multiply(A=ComponentMask(None=@n1) {g=true}, B=ScalarParameter() {defaultValue=5, group=Distortion UV, parameterName=Mask Tile V})
  n4 = Time() {bOverride_Period=true, period=64}
  n5 = DynamicParameter() {defaultValue=<rgba 1 0 1 1>, paramNames [WPO Strength, Erode, Erode Noise Strength, Erode UV Offset]}
  n6 = ScalarParameter() {defaultValue=0.1, group=Distortion UV, parameterName=Mask Pan Speed U}
  n7 = ScalarParameter() {defaultValue=0.5, group=Distortion UV, parameterName=Mask Pan Speed V}
  n8 = TextureSample(UVs=Panner(Coordinate=AppendVector(A=@n2, B=@n3), Time=Add(A=@n4, B=@n5.Erode UV Offset), Speed=AppendVector(A=@n6, B=@n7)) {speedX=-1, speedY=0.1}) {samplerType=SAMPLERTYPE_Masks, texture=/Game/Weapons/WeaponSkins/Wraps/Textures/Mash/Mash_Noise}
  n9 = DynamicParameter() {defaultValue=<rgba 1 0 1 1>, paramNames [WPO Strength, Erode, Erode Noise Strength, Erode UV Offset]}
  Add(
   A=Multiply(A=VertexColor().R, B=Multiply(A=@n8.R, B=@n9.Erode Noise Strength))
   B=DynamicParameter() {defaultValue=<rgba 1 0 1 1>, paramNames [WPO Strength, Erode, Erode Noise Strength, Erode UV Offset]}.Erode
  )
  n10 = VectorParameter() {parameterName=Color Center, defaultValue=<rgba 3 3 3 0>}
  n11 = VectorParameter() {parameterName=Color Dark, defaultValue=<rgba 0 0.3484 1 1>}
  n12 = ScalarParameter() {defaultValue=0.5, parameterName=Fresnel Exponent}
  n13 = Constant()
  n14 = VertexNormalWS()
  n15 = OneMinus(None=ComponentMask(None=TextureCoordinate()) {g=true})
  n16 = ScalarParameter() {defaultValue=0.2, parameterName=Fresnel Mask Offset}
  n17 = Multiply(
   A=MaterialFunctionCall(
    In=Fresnel(ExponentIn=@n12, BaseReflectFractionIn=@n13, Normal=@n14)
    Contrast=ScalarParameter() {defaultValue=0.5, parameterName=Fresnel Contrast}
   ) {materialFunction=&CheapContrast}.Result
   B=MaterialFunctionCall(In=Add(A=@n15, B=@n16), Contrast=ScalarParameter() {defaultValue=0.5, parameterName=Fresnel Mask Contrast}) {materialFunction=&CheapContrast}.Result
  )
  MaterialFunctionCall(Emissive=Multiply(A=ParticleColor().RGB, B=LinearInterpolate(A=@n10.RGB, B=@n11.RGB, Alpha=@n17))) {materialFunction=&MF_FX_EmissiveAdjustments}
  n18 = ScalarParameter() {defaultValue=1, parameterName=WPO Strength}
  n19 = DynamicParameter() {defaultValue=<rgba 1 0 1 1>, paramNames [WPO Strength, Erode, Erode Noise Strength, Erode UV Offset]}
  Multiply(A=Multiply(A=VertexNormalWS(), B=Multiply(A=@n18, B=@n19.WPO Strength)), B=VertexColor().R)
  n20 = VertexColor()
  n21 = ScalarParameter() {parameterName=VertexColor Contrast}
  Multiply(
   A=Multiply(
    A=VertexNormalWS()
    B=MaterialFunctionCall(In=@n20.R, Contrast=@n21) {materialFunction=&CheapContrast}.Result
   )
   B=DynamicParameter() {defaultValue=<rgba 1 0 1 1>, paramNames [WPO Strength, Erode, Erode Noise Strength, Erode UV Offset]}.WPO Strength
  )
 ]
}
