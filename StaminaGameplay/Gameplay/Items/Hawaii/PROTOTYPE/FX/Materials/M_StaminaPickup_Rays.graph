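# Two-sided unlit alpha-composite material for light rays on a prototype stamina pickup effect, usable on Niagara sprites and meshes.
# A length-adjusted sphere mask along flipped V shapes the ray, broken up by two panning noise channels that strengthen where the mask weakens.
# That mask drives opacity and an opacity-premultiplied color, either a two-color gradient or custom primitive data.
# A Sprite function billboards the geometry; medium and high quality add depth and near-camera fades.
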Material /StaminaGameplay/Gameplay/Items/Hawaii/PROTOTYPE/FX/Materials/M_StaminaPickup_Rays
graph {
 properties {
  bEnableResponsiveAA=true
  bUsedWithNiagaraMeshParticles=true
  bUsedWithNiagaraSprites=true
  blendMode=BLEND_AlphaComposite
  pixelDepthOffsetMode=PDOM_Legacy
  shadingModel=MSM_Unlit
  twoSided=true
  usedShadingModels=Unlit
 }
 nodes [
  MaterialFunctionCall!(
   Center Location
   0-1 UVs=TextureCoordinate() {coordinateIndex=1}
   XY scale=ScalarParameter() {defaultValue=35, parameterName=Scale}
   OverridePivot
   "Normals (Optional)"
  ) {materialFunction=&Sprite}
  n1 = ScalarParameter() {defaultValue=1, parameterName=Fade}
  n2 = OneMinus(None=TextureCoordinate())
  n3 = ScalarParameter() {defaultValue=0.9, parameterName=FalloffRadius}
  n4 = Divide(A=ScalarParameter() {defaultValue=300, parameterName=Length}, B=ScalarParameter() {defaultValue=400, parameterName=FalloffBaseLength})
  n5 = SphereMask(
   A=ComponentMask(None=@n2) {g=true}
   B=Constant() {r=1}
   Radius=Saturate(None=Divide(A=@n3, B=@n4))
   Hardness=ScalarParameter() {defaultValue=0.025, parameterName=FalloffHardness}
  )
  n6 = Power(Base=@n5, Exp=ScalarParameter() {defaultValue=3.5, parameterName=MaskExponent})
  n7 = Multiply(A=@n2, B=ComponentMask(None=VectorParameter() {parameterName=NoiseUVTileA, defaultValue=<rgba 2 0.09 0 0>}.RGB) {g=true, r=true})
  n8 = ComponentMask(None=VectorParameter() {parameterName=NoisePanSpeedA, defaultValue=<rgba 0.36 0.11 0 0>}.RGB) {g=true, r=true}
  n9 = Subtract(A=TextureSample(UVs=Panner(Coordinate=@n7, Speed=@n8) {bFractionalPart=true}) {samplerType=SAMPLERTYPE_LinearColor, texture=&T_NoisePack_LinearColor}.R) {constB=0.5}
  n10 = Multiply(A=@n2, B=ComponentMask(None=VectorParameter() {parameterName=NoiseUVTileB, defaultValue=<rgba 3 0.11 0 0>}.RGB) {g=true, r=true})
  n11 = ComponentMask(None=VectorParameter() {parameterName=NoisePanSpeedB, defaultValue=<rgba -0.63 0 0.1335 0>}.RGB) {g=true, r=true}
  n12 = Subtract(A=TextureSample(UVs=Panner(Coordinate=@n10, Speed=@n11) {bFractionalPart=true, speedX=-0.026}) {samplerType=SAMPLERTYPE_LinearColor, texture=&T_NoisePack_LinearColor}.G) {constB=0.5}
  n13 = Multiply(A=Multiply(A=OneMinus(None=@n6), B=Add(A=@n9, B=@n12)), B=ScalarParameter() {defaultValue=1.5, parameterName=NoiseScale})
  n14 = LinearInterpolate(B=ScalarParameter() {defaultValue=0.6, parameterName=OpacityHigh}, Alpha=@n6)
  n15 = Multiply(A=@n5, B=Add(A=@n13, B=@n14))
  n16 = ScalarParameter() {defaultValue=30, parameterName=FadeDistance}
  n17 = DepthFade(Opacity=@n15, FadeDistance=@n16)
  n18 = Constant() {r=1}
  n19 = MaterialFunctionCall(Distance=ScalarParameter() {defaultValue=100, parameterName=NearCameraFadeDistance}) {materialFunction=&Near_Camera_Fade}
  n20 = QualitySwitch(Default=@n18, Low=@n18, High=@n19.Result, Medium=@n19.Result)
  Multiply(A=Multiply(A=@n1, B=QualitySwitch(Default=@n15, Low=@n15, High=@n17, Medium=@n17)), B=@n20)
  n21 = VectorParameter() {bUseCustomPrimitiveData=true, parameterName=CustomPrimitiveColor, primitiveDataIndex=28}
  n22 = Multiply(A=Add(A=@n14, B=@n13), B=@n5)
  n23 = LinearInterpolate(
   A=VectorParameter() {parameterName=ColorLow, defaultValue=<rgba 0.0833 0.0247 0.8 1>}.RGB
   B=VectorParameter() {parameterName=ColorHigh, defaultValue=<rgba 0.1438 0.6079 2 1>}.RGB
   Alpha=Saturate(None=@n22)
  )
  n24 = DepthFade(Opacity=@n22, FadeDistance=@n16)
  n25 = QualitySwitch(Default=@n22, Low=@n22, High=@n24, Medium=@n24)
  Multiply(
   A=Multiply(A=StaticSwitchParameter(True=@n21.RGB, False=@n23) {parameterName=ColorUsesCustomPrimitiveVec4}, B=Multiply(A=@n25, B=@n1))
   B=@n20
  )
 ]
}
