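# Niagara muzzle effect for the prototype HawaiiStamina beam item, combining a blue glow and light, spinning muzzle meshes and pulsing rings.
# All five emitters anchor each update to the midpoint of two skeletal mesh samples, pushed forward along a system-level beam orientation.
# Rings face User.BeamEnd_0 and travel outward while growing and shifting from white-hot to blue.
# A shared system Timeline drives shrink or fade curves in every emitter, winding the effect down together.
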NiagaraSystem /StaminaGameplay/Gameplay/Items/Hawaii/PROTOTYPE/FX/NS_HawaiiStamina_Muzzle
dependencies {
 usedDataInterfaces [
  &NiagaraDataInterfaceCurve
  &NiagaraDataInterfaceVectorCurve
  &NiagaraDataInterfaceColorCurve
  &NiagaraDataInterfaceSkeletalMesh
 ]
 usedDynamicInputs [
  &VectorLength
  &FindQuatBetweenTwoVectors
  &SetBoolByFloatComparison
  &MakeFloatFromInt
 ]
 usedModules [
  &self:SystemSpawnScript.SystemScriptSource.SystemScriptGraph.NiagaraNodeAssignment_2.SetVariables_#1
  &self:SystemSpawnScript.SystemScriptSource.SystemScriptGraph.NiagaraNodeAssignment_3.SetVariables_#2
  &SystemState
  &SampleSkeletalMesh
  &self:SystemSpawnScript.SystemScriptSource.SystemScriptGraph.NiagaraNodeAssignment_0.SetVariables_#3
  &self:SystemSpawnScript.SystemScriptSource.SystemScriptGraph.NiagaraNodeAssignment_1.SetVariables_#4
  &Timeline
 ]
}
emitters {
 Light {
  emitterUpdateScript [
   &EmitterState()
   &SpawnBurst_Instantaneous(Spawn Count=1, Use Spawn Probability=false)
  ]
  particleSpawnScript [&InitializeParticle(
    Color=<rgba 0 0.7453 2 1>
    Lifetime=15
    Mesh Scale Mode=::NewEnumerator3
    Mesh Uniform Scale=4
    Sprite Size Mode=::NewEnumerator3
    Uniform Sprite Size=150
    UsePositionOffset=false
    Write Material Random=true
   )]
  particleUpdateScript [
   &ParticleState()
   &SkeletalMeshLocation(
    Bone Sampling Mode=::NewEnumerator4
    MaskSpawnGroup=false
    Orientation Sampling=::NewEnumerator1
    Sampled Position Offset Coordinate Space:ENiagaraSkelMeshTransforms=::NewEnumerator2
    Skeletal Mesh=$System.Skeletal Mesh
   )
   /SetVariables(Particles.Position=&Add_Vector(
     A=/Niagara/DynamicInputs/Lerp/Lerp_Position(
      A:NiagaraPosition=$System.SampleSkeletalMesh.SampledPosition
      B:NiagaraPosition=$System.SampleSkeletalMesh001.SampledPosition
      Alpha:NiagaraFloat=0.5
     )
     B=&RotateVectorByQuaternion(Vector To Rotate=<50 0 0>, Quaternion=$System.BeamOrientation)
    )) {isSetParameters=true}
   &FN_Light_Attributes(
    Color=$Particles.Color
    Create Color Attribute=true
    Create Exponent Attribute=true
    Create Radius Attribute=true
    Exponent=6
    Radius=300
    Scale Color=&Multiply_Float(
     A=&FloatFromCurve(
      FloatCurve=DataInterface() {propertyValues {Curve {keys [
          {arriveTangent=9.6835, interpMode=RCIM_Cubic, leaveTangent=9.6835, tangentMode=RCTM_User, time=0, value=0}
          {arriveTangent=0.025, interpMode=RCIM_Cubic, leaveTangent=0.025, tangentMode=RCTM_User, time=0.05, value=1}
          {arriveTangent=-0.4638, interpMode=RCIM_Cubic, leaveTangent=-0.4638, tangentMode=RCTM_User, time=0.9, value=1}
         ]}}}
      CurveIndex=$Particles.Age
      Scale Curve=5
     )
     B=&FloatFromCurve(
      FloatCurve=DataInterface() {propertyValues {Curve {keys [{interpMode=RCIM_Cubic, tangentMode=RCTM_User, time=0, value=1}, {interpMode=RCIM_Cubic, tangentMode=RCTM_User, time=1, value=0}]}}}
      CurveIndex=$System.Timeline.NormalizedTime
     )
    )
   )
   &ScaleSpriteSize(Scale Factor=&Vector2DFromFloat(Value=&Multiply_Float(
      A=&FloatFromCurve(
       FloatCurve=DataInterface() {propertyValues {Curve {keys [
           {arriveTangent=9.6835, interpMode=RCIM_Cubic, leaveTangent=9.6835, tangentMode=RCTM_User, time=0, value=0}
           {arriveTangent=0.025, interpMode=RCIM_Cubic, leaveTangent=0.025, tangentMode=RCTM_User, time=0.05, value=1}
          ]}}}
       CurveIndex=$Particles.Age
      )
      B=&FloatFromCurve(
       FloatCurve=DataInterface() {propertyValues {Curve {keys [
           {arriveTangent=-1.2411, interpMode=RCIM_Cubic, leaveTangent=-1.2411, tangentMode=RCTM_User, time=0, value=1}
           {interpMode=RCIM_Cubic, tangentMode=RCTM_User, time=1, value=0}
          ]}}}
       CurveIndex=$System.Timeline.NormalizedTime
      )
     )))
   &ScaleColor(
    Scale Alpha=&Multiply_Float(
     A=&FloatFromCurve(
      FloatCurve=DataInterface() {propertyValues {Curve {keys [
          {arriveTangent=9.6835, interpMode=RCIM_Cubic, leaveTangent=9.6835, tangentMode=RCTM_User, time=0, value=0}
          {arriveTangent=0.025, interpMode=RCIM_Cubic, leaveTangent=0.025, tangentMode=RCTM_User, time=0.05, value=1}
         ]}}}
      CurveIndex=$Particles.Age
     )
     B=&FloatFromCurve(
      FloatCurve=DataInterface() {propertyValues {Curve {keys [
          {arriveTangent=-1.4996, interpMode=RCIM_Cubic, leaveTangent=-1.4996, tangentMode=RCTM_User, time=0, value=1}
          {interpMode=RCIM_Cubic, tangentMode=RCTM_User, time=1, value=0}
         ]}}}
      CurveIndex=$System.Timeline.NormalizedTime
     )
    )
    ScaleA=true
    ScaleRGB=false
   )
   &FN_Sprite_Attributes(
    Create SpriteSize Attribute=true
    SpriteSize=&Multiply_Vector2D_ByFloat(Float=2, Vector2D=$Particles.SpriteSize)
    Write to Module Namespace=true
   )
   &FN_Sprite_Attributes(
    Color=<rgba 0 0.0003 0.0417 1>
    Create Color Attribute=true
    Create SpriteSize Attribute=true
    SpriteSize=&Multiply_Vector2D_ByFloat(Float=2.5, Vector2D=$Particles.SpriteSize)
    Write to Module Namespace=true
   ) {module=FN_Sprite_Attributes001}
   &ScaleMeshSize(Scale Factor=&VectorFromFloat(Value=&Multiply_Float(
      A=&FloatFromCurve(
       FloatCurve=DataInterface() {propertyValues {Curve {keys [
           {arriveTangent=9.6835, interpMode=RCIM_Cubic, leaveTangent=9.6835, tangentMode=RCTM_User, time=0, value=0}
           {arriveTangent=0.025, interpMode=RCIM_Cubic, leaveTangent=0.025, tangentMode=RCTM_User, time=0.05, value=1}
          ]}}}
       CurveIndex=$Particles.Age
       Scale Curve=1.5
      )
      B=&FloatFromCurve(
       FloatCurve=DataInterface() {propertyValues {Curve {keys [
           {arriveTangent=-1.01, interpMode=RCIM_Cubic, leaveTangent=-1.01, tangentMode=RCTM_User, time=0, value=1}
           {interpMode=RCIM_Cubic, tangentMode=RCTM_User, time=1, value=0}
          ]}}}
       CurveIndex=$System.Timeline.NormalizedTime
      )
     )))
   &CameraOffset(Camera Offset Amount=10) {module=CameraOffset001}
  ]
  rendererClasses [
   &NiagaraSpriteRendererProperties
   &NiagaraLightRendererProperties
   &NiagaraMeshRendererProperties
  ]
  renderers [
   {rendererClass=&NiagaraSpriteRendererProperties}
   {rendererClass=&NiagaraLightRendererProperties, rendererIndex=1}
   {rendererClass=&NiagaraSpriteRendererProperties, rendererIndex=2}
   {rendererClass=&NiagaraMeshRendererProperties, rendererIndex=3}
   {rendererClass=&NiagaraSpriteRendererProperties, rendererIndex=4}
  ]
 }
 Muzzle_MOving {
  emitterUpdateScript [
   &EmitterState()
   &SpawnBurst_Instantaneous(Use Spawn Probability=false, UseLoopCountLimit=false)
   &SpawnRate(SpawnRate=7, Use Spawn Probability=false)
  ]
  particleSpawnScript [
   &InitializeParticle(Color=<rgba 2 2 2 1>, Mesh Scale Mode=::NewEnumerator3, Mesh Uniform Scale=2.8, UsePositionOffset=false, Write Material Random=true)
   &InitialMeshOrientation(
    Orientation Vector=<-1 0 0>
    Rotation=&RandomRangeVector(Minimum=<-1 0 0>, Maximum=<1 0 0>)
    Use Orientation Vector=true
    Use Rotation Vector=true
   )
  ]
  particleUpdateScript [
   &ParticleState()
   /SetVariables(Particles.Position=&Lerp_Position(
     A:NiagaraPosition=&Add_Vector(
      A=&Lerp_Position(
       A:NiagaraPosition=$System.SampleSkeletalMesh.SampledPosition
       B:NiagaraPosition=$System.SampleSkeletalMesh001.SampledPosition
       Alpha:NiagaraFloat=0.5
      )
      B=&RotateVectorByQuaternion(Vector To Rotate=<50 0 0>, Quaternion=$System.BeamOrientation)
     )
     B:NiagaraPosition=&Add_Vector(
      A=&Lerp_Position(
       A:NiagaraPosition=$System.SampleSkeletalMesh.SampledPosition
       B:NiagaraPosition=$System.SampleSkeletalMesh001.SampledPosition
       Alpha:NiagaraFloat=0.5
      )
      B=&RotateVectorByQuaternion(
       Vector To Rotate=&MakeVector(
        X=&Multiply_Float(
         A=&FloatFromCurve(
          FloatCurve=DataInterface() {propertyValues {Curve {keys [
              {arriveTangent=9.6835, interpMode=RCIM_Cubic, leaveTangent=9.6835, tangentMode=RCTM_User, time=0, value=0}
              {arriveTangent=0.025, interpMode=RCIM_Cubic, leaveTangent=0.025, tangentMode=RCTM_User, time=0.1, value=1}
             ]}}}
          CurveIndex=$Emitter.Age
          Scale Curve=150
         )
         B=&FloatFromCurve(
          FloatCurve=DataInterface() {propertyValues {Curve {keys [
              {arriveTangent=-0.9518, interpMode=RCIM_Cubic, leaveTangent=-0.9518, tangentMode=RCTM_User, time=0, value=1}
              {interpMode=RCIM_Cubic, tangentMode=RCTM_User, time=1, value=0}
             ]}}}
          CurveIndex=$System.Timeline.NormalizedTime
         )
        )
        Y=0
        Z=0
       )
       Quaternion=$System.BeamOrientation
      )
     )
     Alpha:NiagaraFloat=$Particles.NormalizedAge
    )) {isSetParameters=true}
   &UpdateMeshOrientation(
    Facing Direction=&NormalizeVector(
     Vector To Normalize=&RotateVectorByQuaternion(Vector To Rotate=<1 0 0>, Quaternion=$System.BeamOrientation)
     Safe Normalize=true
     Safe Fallback Threshold=0
     Safe Fallback Vector=<0 0 0.001>
    )
    Facing Interpolation=false
    Orientation Method=::NewEnumerator1
    Override Quat=false
    TransformFacing=false
    TransformUp=false
   )
   &UpdateMeshOrientation(Rotation Rate=0.35) {module=UpdateMeshOrientation001}
   /SetVariables(Particles.R2_DynMat2:Vector4f=&MakeVector4(X=10, Y=0, Z=0, W=0)) {isSetParameters=true}
   &ScaleMeshSize(Scale Factor=&VectorFromCurve(
     VectorCurve=DataInterface() {propertyValues {XCurve {keys [<at 0 1>,<at 1 2>]}, YCurve {keys [<at 0 1>,<at 1 0>]}, ZCurve {keys [<at 0 1>,<at 1 0>]}}}
     Scale Curve=&VectorFromFloat(Value=&Multiply_Float(
       A=&FloatFromCurve(
        FloatCurve=DataInterface() {propertyValues {Curve {keys [
            {arriveTangent=9.6835, interpMode=RCIM_Cubic, leaveTangent=9.6835, tangentMode=RCTM_User, time=0, value=0}
            {arriveTangent=0.025, interpMode=RCIM_Cubic, leaveTangent=0.025, tangentMode=RCTM_User, time=0.1, value=1}
           ]}}}
        CurveIndex=$Emitter.Age
       )
       B=&FloatFromCurve(
        FloatCurve=DataInterface() {propertyValues {Curve {keys [
            {arriveTangent=-1.5257, interpMode=RCIM_Cubic, leaveTangent=-1.5257, tangentMode=RCTM_User, time=0, value=1}
            {interpMode=RCIM_Cubic, tangentMode=RCTM_User, time=1, value=0}
           ]}}}
        CurveIndex=$System.Timeline.NormalizedTime
       )
      ))
    ))
   &CameraOffset(Camera Offset Amount=10)
  ]
  rendererClasses [&NiagaraMeshRendererProperties]
  renderers [{rendererClass=&NiagaraMeshRendererProperties, rendererIndex=0}]
 }
 Muzzle_spinny001 {
  emitterUpdateScript [
   &EmitterState()
   &SpawnBurst_Instantaneous(Use Spawn Probability=false, UseLoopCountLimit=false)
  ]
  particleSpawnScript [
   &InitializeParticle(
    Color=<rgba 2 2 2 1>
    Lifetime=15
    Mesh Scale Mode=::NewEnumerator3
    Mesh Uniform Scale=2.8
    UsePositionOffset=false
    Write Material Random=true
   )
   &InitialMeshOrientation(
    Orientation Axis=<-1 0 0>
    Rotation=&RandomRangeVector(Minimum=<-1 0 0>, Maximum=<1 0 0>)
    Use Orientation Vector=true
    Use Rotation Vector=true
   )
  ]
  particleUpdateScript [
   &ParticleState()
   /SetVariables(Particles.Position=&Add_Vector(
     A=&Lerp_Position(
      A:NiagaraPosition=$System.SampleSkeletalMesh.SampledPosition
      B:NiagaraPosition=$System.SampleSkeletalMesh001.SampledPosition
      Alpha:NiagaraFloat=0.5
     )
     B=&RotateVectorByQuaternion(Vector To Rotate=<50 0 0>, Quaternion=$System.BeamOrientation)
    )) {isSetParameters=true, module=SetVariables_#5}
   &UpdateMeshOrientation(
    Facing Direction=&NormalizeVector(
     Vector To Normalize=&RotateVectorByQuaternion(Vector To Rotate=<1 0 0>, Quaternion=$System.BeamOrientation)
     Safe Normalize=true
     Safe Fallback Threshold=0
     Safe Fallback Vector=<0 0 0.001>
    )
    Facing Interpolation=false
    Orientation Method=::NewEnumerator1
    Override Quat=false
    TransformFacing=false
    TransformUp=false
   )
   &UpdateMeshOrientation(Rotation Rate=0.35) {module=UpdateMeshOrientation001}
   &ScaleMeshSize(Scale Factor=&VectorFromFloat(Value=&Multiply_Float(
      A=&FloatFromCurve(
       FloatCurve=DataInterface() {propertyValues {Curve {keys [
           {arriveTangent=9.6835, interpMode=RCIM_Cubic, leaveTangent=9.6835, tangentMode=RCTM_User, time=0, value=0}
           {arriveTangent=0.025, interpMode=RCIM_Cubic, leaveTangent=0.025, tangentMode=RCTM_User, time=0.05, value=1}
          ]}}}
       CurveIndex=$Particles.Age
      )
      B=&FloatFromCurve(
       FloatCurve=DataInterface() {propertyValues {Curve {keys [
           {arriveTangent=-1.9755, interpMode=RCIM_Cubic, leaveTangent=-1.9755, tangentMode=RCTM_User, time=0, value=1}
           {interpMode=RCIM_Cubic, tangentMode=RCTM_User, time=1, value=0}
          ]}}}
       CurveIndex=$System.Timeline.NormalizedTime
      )
     )))
   &CameraOffset(Camera Offset Amount=10)
  ]
  rendererClasses [&NiagaraMeshRendererProperties]
  renderers [{rendererClass=&NiagaraMeshRendererProperties, rendererIndex=0}]
 }
 Muzzle_spinny_scaling {
  emitterUpdateScript [
   &EmitterState()
   &SpawnBurst_Instantaneous(Use Spawn Probability=false, UseLoopCountLimit=false)
   &SpawnRate(SpawnRate=7, Use Spawn Probability=false)
  ]
  particleSpawnScript [
   &InitializeParticle(
    Color=<rgba 2 2 2 1>
    Lifetime=0.75
    Mesh Scale Mode=::NewEnumerator3
    Mesh Uniform Scale=2.5
    UsePositionOffset=false
    Write Material Random=true
   )
   &InitialMeshOrientation(
    Orientation Axis=<-1 0 0>
    Rotation=&RandomRangeVector(Minimum=<-1 0 0>, Maximum=<1 0 0>)
    Use Orientation Vector=true
    Use Rotation Vector=true
   )
  ]
  particleUpdateScript [
   &ParticleState()
   /SetVariables(Particles.Position=&Add_Vector(
     A=&Lerp_Position(
      A:NiagaraPosition=$System.SampleSkeletalMesh.SampledPosition
      B:NiagaraPosition=$System.SampleSkeletalMesh001.SampledPosition
      Alpha:NiagaraFloat=0.5
     )
     B=&RotateVectorByQuaternion(Vector To Rotate=<50 0 0>, Quaternion=$System.BeamOrientation)
    )) {isSetParameters=true, module=SetVariables_#5}
   &UpdateMeshOrientation(
    Facing Direction=&NormalizeVector(
     Vector To Normalize=&RotateVectorByQuaternion(Vector To Rotate=<1 0 0>, Quaternion=$System.BeamOrientation)
     Safe Normalize=true
     Safe Fallback Threshold=0
     Safe Fallback Vector=<0 0 0.001>
    )
    Facing Interpolation=false
    Orientation Method=::NewEnumerator1
    Override Quat=false
    TransformFacing=false
    TransformUp=false
   )
   &UpdateMeshOrientation(Rotation Rate=0.35) {module=UpdateMeshOrientation001}
   &ScaleMeshSize(Scale Factor=&VectorFromFloat(Value=&FloatFromCurve(
      FloatCurve=DataInterface() {propertyValues {Curve {keys [
          {arriveTangent=2.3327, interpMode=RCIM_Cubic, leaveTangent=2.3327, tangentMode=RCTM_User, time=0, value=0}
          {interpMode=RCIM_Cubic, tangentMode=RCTM_User, time=1, value=1}
         ]}}}
      Scale Curve=&FloatFromCurve(
       FloatCurve=DataInterface() {propertyValues {Curve {keys [
           {arriveTangent=-1.1917, interpMode=RCIM_Cubic, leaveTangent=-1.1917, tangentMode=RCTM_User, time=0, value=1}
           {interpMode=RCIM_Cubic, tangentMode=RCTM_User, time=1, value=0}
          ]}}}
       CurveIndex=$System.Timeline.NormalizedTime
      )
     )))
   &CameraOffset(Camera Offset Amount=15)
   &ScaleColor(
    Scale Alpha=&FloatFromCurve(FloatCurve=DataInterface() {propertyValues {Curve {keys [
         {arriveTangent=6.2774, interpMode=RCIM_Cubic, leaveTangent=6.2774, tangentMode=RCTM_User, time=0, value=0}
         <at 0.4 1 RCIM_Cubic>
         <at 1 0 RCIM_Cubic>
        ]}}})
    ScaleA=true
    ScaleRGB=true
   )
  ]
  rendererClasses [&NiagaraMeshRendererProperties]
  renderers [{rendererClass=&NiagaraMeshRendererProperties, rendererIndex=0}]
 }
 Rings {
  emitterUpdateScript [
   &EmitterState()
   &SpawnBurst_Instantaneous(Spawn Count=1, Use Spawn Probability=false)
   &SpawnRate(SpawnRate=4, Use Spawn Probability=false)
   &FN_Timer()
  ]
  particleSpawnScript [
   &InitializeParticle(
    Lifetime Max=0.7
    Lifetime Min=0.6
    Lifetime Mode=::NewEnumerator1
    Mesh Scale Mode=::NewEnumerator3
    Mesh Uniform Scale=8
    Position Offset=<8 0 0>
    UsePositionOffset=true
    Write Material Random=true
   ) {module=InitializeParticle001}
   &InitialMeshOrientation(
    Rotation=&RandomRangeVector(Minimum=<-1 0 0>, Maximum=<1 0 0>)
    Use Orientation Vector=true
    Use Rotation Vector=true
   )
  ]
  particleUpdateScript [
   &ParticleState()
   &SkeletalMeshLocation(
    Bone Sampling Mode=::NewEnumerator4
    MaskSpawnGroup=false
    Orientation Sampling=::NewEnumerator1
    Sampled Position Offset Coordinate Space:ENiagaraSkelMeshTransforms=::NewEnumerator2
    Skeletal Mesh=$System.Skeletal Mesh
   )
   &UpdateMeshOrientation(
    Facing Direction=&NormalizeVector(
     Vector To Normalize=&Subtract_Vector(A=Linked() {linkedVariable {name=User.BeamEnd_0, type=NiagaraPosition}}, B=$Particles.SkeletalMeshLocation.SampledPosition)
     Safe Normalize=true
     Safe Fallback Threshold=0
     Safe Fallback Vector=<0 0 0.001>
    )
    Facing Interpolation=false
    Orientation Method=::NewEnumerator1
    Override Quat=false
    TransformFacing=false
    TransformUp=false
   )
   /SetVariables(Particles.Offset:Vector3f=&RotateVectorByQuaternion(
     Vector To Rotate=&MakeVector(
      X=&FloatFromCurve(
       FloatCurve=DataInterface() {propertyValues {Curve {keys [{interpMode=RCIM_Cubic, tangentMode=RCTM_User, time=0, value=0}, <at 0.2 0>, {interpMode=RCIM_Cubic, tangentMode=RCTM_User, time=1, value=1}]}}}
       Scale Curve=150
      )
      Y=0
      Z=0
     )
     Quaternion=$Particles.MeshOrientation
    )) {isSetParameters=true}
   /SetVariables(Particles.Position=&ConvertVectorToPosition(
     Input Position=&Add_Vector(
      A=&ConvertPositionToVector(
       Input Position=&Add_Vector(
        A=&Lerp_Position(
         A:NiagaraPosition=$System.SampleSkeletalMesh.SampledPosition
         B:NiagaraPosition=$System.SampleSkeletalMesh001.SampledPosition
         Alpha:NiagaraFloat=0.5
        )
        B=&RotateVectorByQuaternion(Vector To Rotate=<50 0 0>, Quaternion=$System.BeamOrientation)
       )
       Convert To Absolute Large World Vector=true
      )
      B=$Particles.Offset
     )
     Vector Is In Absolute Large World Space:NiagaraBool=true
    )) {isSetParameters=true}
   &DynamicMaterialParameters(
    Index 2 Param 4=1
    Index 3 Param 1=&Add_Float(A=&RandomRangeFloat(Minimum=0, Maximum=1), B=$Emitter.FN_Timer.Value)
    Index 3 Param 2=&FloatFromCurve(FloatCurve=DataInterface() {propertyValues {Curve {keys [
         {interpMode=RCIM_Cubic, tangentMode=RCTM_User, time=0, value=1}
         {interpMode=RCIM_Cubic, tangentMode=RCTM_User, time=0.3, value=1}
         {arriveTangent=-2.8641, interpMode=RCIM_Cubic, leaveTangent=-2.8641, tangentMode=RCTM_User, time=1, value=0}
        ]}}})
    Module.Param0WriteEnabled2=true
    Module.Param0WriteEnabled3=true
    Module.Param1WriteEnabled2=true
    Module.Param1WriteEnabled3=true
    Module.Param2WriteEnabled2=true
    Module.Param2WriteEnabled3=true
    Module.Param3WriteEnabled2=true
    Module.Param3WriteEnabled3=true
    Write Parameter Index 0=false
    Write Parameter Index 2=true
    Write Parameter Index 3=true
   )
   &ScaleColor(
    Scale Alpha=&FloatFromCurve(FloatCurve=DataInterface() {propertyValues {Curve {keys [<at 0 1>,<at 1 0>]}}}, CurveIndex=$System.Timeline.NormalizedTime)
    Scale RGB=&MakeVectorFromLinearColorRGB(Color=&ColorFromCurve(ColorCurve=DataInterface() {propertyValues {
        AlphaCurve {keys [<at 0 1 RCIM_Cubic>]}
        BlueCurve {keys [<at 0 5 RCIM_Cubic>,<at 0.2184 2.5>,<at 0.6752 0.6042 RCIM_Cubic>]}
        GreenCurve {keys [<at 0 5 RCIM_Cubic>,<at 0.2184 1.545>,<at 0.6752 0.2251 RCIM_Cubic>]}
        RedCurve {keys [<at 0 5 RCIM_Cubic>,<at 0.2184 0.3765>,<at 0.6752 0 RCIM_Cubic>]}
       }}))
    ScaleA=true
    ScaleRGB=true
   )
   &ScaleMeshSize(Scale Factor=&VectorFromCurve(VectorCurve=DataInterface() {propertyValues {
       XCurve {keys [
         {arriveTangent=2.2871, interpMode=RCIM_Cubic, leaveTangent=2.2871, tangentMode=RCTM_User, time=0, value=0}
         {arriveTangent=1, interpMode=RCIM_Cubic, leaveTangent=1, time=0.5, value=0.7}
         <at 1 1 RCIM_Cubic>
        ]}
       YCurve {keys [
         {arriveTangent=2.2871, interpMode=RCIM_Cubic, leaveTangent=2.2871, tangentMode=RCTM_User, time=0, value=0}
         {arriveTangent=1, interpMode=RCIM_Cubic, leaveTangent=1, time=0.5, value=0.7}
         <at 1 1 RCIM_Cubic>
        ]}
       ZCurve {keys [
         {arriveTangent=2.2871, interpMode=RCIM_Cubic, leaveTangent=2.2871, tangentMode=RCTM_User, time=0, value=0}
         {arriveTangent=1, interpMode=RCIM_Cubic, leaveTangent=1, time=0.5, value=0.7}
         <at 1 1 RCIM_Cubic>
        ]}
      }}))
  ]
  rendererClasses [&NiagaraMeshRendererProperties]
  renderers [{rendererClass=&NiagaraMeshRendererProperties, rendererIndex=0}]
 }
}
summary {
 systemName=NS_HawaiiStamina_Muzzle
 emitters [
  {
   emitterName=Light
   rendererClasses [
    &NiagaraSpriteRendererProperties
    &NiagaraLightRendererProperties
    &NiagaraMeshRendererProperties
   ]
  }
  {emitterName=Muzzle_spinny001, rendererClasses [&NiagaraMeshRendererProperties]}
  {emitterName=Muzzle_MOving, rendererClasses [&NiagaraMeshRendererProperties]}
  {emitterName=Muzzle_spinny_scaling, rendererClasses [&NiagaraMeshRendererProperties]}
  {emitterName=Rings, rendererClasses [&NiagaraMeshRendererProperties]}
 ]
 userVariables [
  User.BeamEnd_0 {type=NiagaraPosition, description="", defaultValue {}}
  User.BeamFadeOut {type=NiagaraInt32}
  User.BeamStart_0_0 {type=Vector3f, description="", defaultValue {struct=Vector3f, value=<0 0 0>}}
  User.Scale {type=NiagaraFloat, defaultValue=0.2, description=""}
 ]
}
