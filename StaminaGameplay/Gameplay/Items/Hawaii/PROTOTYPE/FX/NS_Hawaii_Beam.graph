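# Niagara beam effect for the prototype Hawaii item, drawing a ribbon beam from a skeletal mesh to a user-supplied end point.
# One burst of partitioned particles is laid along a Hermite spline that starts at an oriented offset from the midpoint of two sampled mesh positions.
# Beam length and age curves shape ribbon width and UVs, while a system timeline fades width, color alpha and a material parameter.
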NiagaraSystem /StaminaGameplay/Gameplay/Items/Hawaii/PROTOTYPE/FX/NS_Hawaii_Beam
dependencies {
 usedDataInterfaces [&NiagaraDataInterfaceCurve]
 usedDynamicInputs [
  &FindQuatBetweenTwoVectors
  &SetBoolByFloatComparison
  &MakeFloatFromInt
 ]
 usedModules [
  &self:SystemSpawnScript.SystemScriptSource.SystemScriptGraph.NiagaraNodeAssignment_2.SetVariables_#1
  &self:SystemSpawnScript.SystemScriptSource.SystemScriptGraph.NiagaraNodeAssignment_3.SetVariables_#2
  &SystemState
  &SampleSkeletalMesh
  &self:SystemSpawnScript.SystemScriptSource.SystemScriptGraph.NiagaraNodeAssignment_0.SetVariables_#3
  &self:SystemSpawnScript.SystemScriptSource.SystemScriptGraph.NiagaraNodeAssignment_1.SetVariables_#4
  &Timeline
 ]
}
emitters {ProjectileBeam {
  emitterSpawnScript [/Game/Effects/Niagara/Modules/Partition/FN_PartitionAttributes(Number of Bone Sets=1, Particles Per Partition=30)]
  emitterUpdateScript [
   &EmitterState()
   &SpawnBurst_Instantaneous(Spawn Count=$Emitter.FN_PartitionAttributes.Spawn Total, Use Spawn Probability=false)
  ]
  particleSpawnScript [
   /Niagara/Modules/Spawn/Initialization/InitializeParticle(
    Lifetime=15
    Module.Write Color=true
    Module.Write Lifetime=true
    Module.Write Mass=true
    Module.Write Material Random=true
    Module.Write Position=true
    Module.Write Scale=false
    Module.Write SpriteRotation=false
    Module.Write SpriteSize=true
    Sprite Size=<30 30>
   )
   &InitializeRibbon(
    Color=<rgba 1 1 1 0.15>
    Lifetime=15
    Module.Write Color=true
    Module.Write Lifetime=true
    Module.Write Mass=true
    Module.Write Material Random=true
    Module.Write Position=false
    Module.Write RibbonTwist=false
    Module.Write RibbonWidth=true
    Ribbon Width=300
   )
   /Game/Effects/Niagara/Modules/Partition/FN_Particle_partition_v5/FN_Particle_Partition_v5(Enable Ribbon ID=true)
   /SetVariables(Particles.Beam2_Width=45) {enabled=false, isSetParameters=true}
   /SetVariables(Particles.WobbleOverLength:NiagaraFloat=&FloatFromCurve(
     FloatCurve=DataInterface() {propertyValues {Curve {keys [
         {arriveTangent=5.2404, interpMode=RCIM_Cubic, leaveTangent=5.2404, tangentMode=RCTM_User, time=0, value=0}
         <at 0.5 1 RCIM_Cubic>
         {arriveTangent=-4.4363, interpMode=RCIM_Cubic, leaveTangent=-4.4363, tangentMode=RCTM_User, time=1, value=0}
        ]}}}
     CurveIndex=$Particles.Partition.NormalizedID
    )) {enabled=false, isSetParameters=true}
  ]
  particleUpdateScript [
   &ParticleState()
   &SystemLocation(MaskSpawnGroup=false) {enabled=false}
   &SkeletalMeshLocation(
    Bone Sampling Mode=::NewEnumerator4
    MaskSpawnGroup=false
    Orientation Sampling=::NewEnumerator1
    Skeletal Mesh=$User.New Skeletal Mesh
    Write To Particle Position=false
   )
   /Game/Effects/Niagara/Modules/Curves/FN_HermiteCubicSpline_v1/FN_HermiteCubicSpline(
    Position Blend:NiagaraFloat=$Particles.Initial.Partition.NormalizedID
    Position End:NiagaraPosition=$User.BeamEnd_0
    Position Start:Vector3f=&Add_Vector(
     A=/Niagara/DynamicInputs/Lerp/Lerp_Position(
      A:NiagaraPosition=$System.SampleSkeletalMesh.SampledPosition
      B:NiagaraPosition=$System.SampleSkeletalMesh001.SampledPosition
      Alpha:NiagaraFloat=0.5
     )
     B=&RotateVectorByQuaternion(Vector To Rotate=<60 0 0>, Quaternion=$System.BeamOrientation)
    )
    Write Particles Position=true
   )
   /SetVariables() {isSetParameters=true}
   /SetVariables(Particles.Ribbon4_Color:LinearColor=<rgba 1 1 1 0.85>) {isSetParameters=true}
   /SetVariables(Particles.BeamLength:NiagaraFloat=&VectorLength(Vector=&Subtract_Vector(
      A=&Lerp_Position(
       A:NiagaraPosition=$Particles.SkeletalMeshLocation.SampledPosition
       B:NiagaraPosition=&ConvertVectorToPosition(
        Input Position=Linked() {linkedVariable {name=User.BeamEnd_0, type=NiagaraPosition}}
        Vector Is In Absolute Large World Space:NiagaraBool=true
       )
       Alpha:NiagaraFloat=&ClampFloat(Float=$Particles.NormalizedAge)
      )
      B=&ConvertPositionToVector(Input Position=$Particles.SkeletalMeshLocation.SampledPosition, Convert To Absolute Large World Vector=true)
     ))) {isSetParameters=true}
   /SetVariables(Particles.FinalBeamLength:NiagaraFloat=&VectorLength(Vector=&Subtract_Vector(
      A=Linked() {linkedVariable {name=Particles.SkeletalMeshLocation.SampledPosition, type=NiagaraPosition}}
      B=Linked() {linkedVariable {name=User.BeamEnd_0, type=NiagaraPosition}}
     ))) {isSetParameters=true}
   &ScaleRibbonWidth(Ribbon Width Scale=&FloatFromCurve(
     FloatCurve=DataInterface() {propertyValues {Curve {keys [
         {arriveTangent=15.5299, interpMode=RCIM_Cubic, leaveTangent=15.5299, tangentMode=RCTM_User, time=0, value=0}
         <at 0.07 1>
         {arriveTangent=-0.1817, interpMode=RCIM_Cubic, leaveTangent=-0.1817, tangentMode=RCTM_User, time=0.5, value=1.05}
        ]}}}
     CurveIndex=&Multiply_Float(A=$Particles.Partition.NormalizedID, B=&Multiply_Float(A=$Particles.BeamLength, B=0.0003))
     Scale Curve=&FloatFromCurve(FloatCurve=DataInterface() {propertyValues {Curve {keys [
          {arriveTangent=9.2323, interpMode=RCIM_Cubic, leaveTangent=9.2323, tangentMode=RCTM_User, time=0, value=0}
          {arriveTangent=-0.1936, interpMode=RCIM_Cubic, leaveTangent=-0.1936, tangentMode=RCTM_User, time=0.1, value=1}
          {arriveTangent=0.4337, interpMode=RCIM_Cubic, leaveTangent=0.4337, tangentMode=RCTM_User, time=0.9, value=1}
          <at 1 0 RCIM_Cubic>
         ]}}})
    )) {enabled=false}
   &ScaleRibbonWidth(Ribbon Width Scale=&Multiply_Float(
     A=&FloatFromCurve(
      FloatCurve=DataInterface() {propertyValues {Curve {keys [
          {arriveTangent=15.5299, interpMode=RCIM_Cubic, leaveTangent=15.5299, tangentMode=RCTM_User, time=0, value=0}
          <at 0.07 1>
          {arriveTangent=-0.1817, interpMode=RCIM_Cubic, leaveTangent=-0.1817, tangentMode=RCTM_User, time=0.5, value=1.05}
         ]}}}
      CurveIndex=&Multiply_Float(A=$Particles.Partition.NormalizedID, B=&Multiply_Float(A=$Particles.FinalBeamLength, B=0.0003))
     )
     B=&Multiply_Float(
      A=&FloatFromCurve(
       FloatCurve=DataInterface() {propertyValues {Curve {keys [
           {arriveTangent=9.2323, interpMode=RCIM_Cubic, leaveTangent=9.2323, tangentMode=RCTM_User, time=0, value=0.5}
           {arriveTangent=-0.1936, interpMode=RCIM_Cubic, leaveTangent=-0.1936, tangentMode=RCTM_User, time=0.015, value=1}
          ]}}}
       CurveIndex=$Particles.Age
      )
      B=&FloatFromCurve(
       FloatCurve=DataInterface() {propertyValues {Curve {keys [{arriveTangent=-0.7446, interpMode=RCIM_Cubic, leaveTangent=-0.7446, tangentMode=RCTM_User, time=0, value=1}, <at 1 0 RCIM_Cubic>]}}}
       CurveIndex=$System.Timeline.NormalizedTime
      )
     )
    )) {module=ScaleRibbonWidth001}
   /SetVariables(Particles.RibbonUVDistance:NiagaraFloat=&Multiply_Float(A=$Particles.Partition.NormalizedID, B=&Multiply_Float(A=$Particles.BeamLength, B=0.0005))) {isSetParameters=true}
   /SetVariables(Particles.Ribbon2_Width:NiagaraFloat=&Multiply_Float(A=$Particles.RibbonWidth, B=0.4)) {isSetParameters=true}
   /SetVariables(Particles.Ribbon3_Width:NiagaraFloat=&Multiply_Float(A=$Particles.RibbonWidth, B=1.2)) {isSetParameters=true}
   /SetVariables(Particles.Ribbon4_Width:NiagaraFloat=&Multiply_Float(A=$Particles.RibbonWidth, B=0.5)) {isSetParameters=true}
   /SetVariables(Particles.Ribbon3_Color:LinearColor=&MultiplyLinearColor(
     Linear Color=<rgba 0 0.3442 0.6892 1>
     Scale Factor=&MakeVector4(
      X=1
      Y=1
      Z=1
      W=&Multiply_Float(
       A=&FloatFromCurve(
        FloatCurve=DataInterface() {propertyValues {Curve {keys [
            {arriveTangent=9.2323, interpMode=RCIM_Cubic, leaveTangent=9.2323, tangentMode=RCTM_User, time=0, value=0}
            {arriveTangent=-0.1936, interpMode=RCIM_Cubic, leaveTangent=-0.1936, tangentMode=RCTM_User, time=0.05, value=1}
            {arriveTangent=0.4337, interpMode=RCIM_Cubic, leaveTangent=0.4337, tangentMode=RCTM_User, time=0.93, value=1}
           ]}}}
        CurveIndex=$Particles.Age
       )
       B=&FloatFromCurve(
        FloatCurve=DataInterface() {propertyValues {Curve {keys [
            {arriveTangent=0.4337, interpMode=RCIM_Cubic, leaveTangent=0.4337, tangentMode=RCTM_User, time=0, value=1}
            {arriveTangent=-0.1936, interpMode=RCIM_Cubic, leaveTangent=-0.1936, tangentMode=RCTM_User, time=0.1, value=1}
            <at 1 0 RCIM_Cubic>
           ]}}}
        CurveIndex=$System.Timeline.NormalizedTime
       )
      )
     )
    )) {isSetParameters=true}
   &FN_RibbonMaterial(Color=<rgba 0 0.5589 1.5 1>, Opacity Pan U Speed=1.5, Write to Module Namespace=true)
   &DynamicMaterialParameters(
    Index 1 Param 1=&FloatFromCurve(
     FloatCurve=DataInterface() {propertyValues {Curve {keys [{arriveTangent=-1.9033, interpMode=RCIM_Cubic, leaveTangent=-1.9033, tangentMode=RCTM_User, time=0, value=1}, <at 1 -1 RCIM_Cubic>]}}}
     CurveIndex=$System.Timeline.NormalizedTime
    )
    Module.Param0WriteEnabled1=true
    Module.Param1WriteEnabled1=true
    Module.Param2WriteEnabled1=true
    Module.Param3WriteEnabled1=true
    Write Parameter Index 0=false
    Write Parameter Index 1=true
   )
   &CameraOffset(Camera Offset Amount=-25) {enabled=false, module=CameraOffset001}
  ]
  rendererClasses [&NiagaraRibbonRendererProperties]
  renderers [
   {rendererClass=&NiagaraRibbonRendererProperties}
   {rendererClass=&NiagaraRibbonRendererProperties, rendererIndex=1}
  ]
 }}
summary {
 systemName=NS_Hawaii_Beam
 emitters [{emitterName=ProjectileBeam, rendererClasses [&NiagaraRibbonRendererProperties]}]
 userVariables [
  User.BeamEnd_0 {type=NiagaraPosition, description="", defaultValue {}}
  User.BeamFadeOut {type=NiagaraInt32}
  User.BeamStart_0_0 {type=Vector3f, description="", defaultValue {struct=Vector3f, value=<0 0 0>}}
  User.New Skeletal Mesh {
   type=NiagaraDataInterfaceSkeletalMesh
   defaultValue {
    struct=NiagaraExt_VariableValue_DataInterface
    value {dataInterface=&self:NiagaraDataInterfaceSkeletalMesh_0}
   }
  }
  User.Scale {type=NiagaraFloat, defaultValue=0.2, description=""}
 ]
}
