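# Niagara charge-up effect for the prototype Hawaii beam item, building for about three seconds at the character's hands.
# Magic sprites, shrinking rings, firework meshes, counter-rotating rays and a glow with a light make up the layers; Rays_LWE is disabled.
# Every active layer follows a Timeline-driven lerp between a rotated offset from one skeletal mesh sample and the midpoint of two samples.
# System-age curves ramp several layers up, and some cut off near the end.
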NiagaraSystem /StaminaGameplay/Gameplay/Items/Hawaii/PROTOTYPE/FX/NS_Hawaii_Beam_Charge
dependencies {
 usedDataInterfaces [
  &NiagaraDataInterfaceCurve
  &NiagaraDataInterfaceColorCurve
  &NiagaraDataInterfaceVectorCurve
  &NiagaraDataInterfaceArrayFloat
 ]
 usedDynamicInputs [&SetBoolByIntComparison]
 usedModules [
  &SystemState
  &SampleSkeletalMesh
  &Timeline
  &self:SystemSpawnScript.SystemScriptSource.SystemScriptGraph.NiagaraNodeAssignment_2.SetVariables_#1
 ]
}
emitters {
 Fireworks_Mesh {
  emitterUpdateScript [
   &EmitterState()
   &SpawnRate(SpawnRate=14, Use Spawn Probability=false)
  ]
  particleSpawnScript [
   &InitializeParticle(
    Color=<rgba 1 1 1 0.3>
    Lifetime=0.25
    Mesh Scale Mode=::NewEnumerator3
    Mesh Uniform Scale=0.3
    UsePositionOffset=false
    Write Material Random=true
   ) {module=InitializeParticle001}
   &InitialMeshOrientation(Use Rotation Vector=false)
  ]
  particleUpdateScript [
   &ParticleState()
   /SetVariables(Particles.Position=/Niagara/DynamicInputs/Lerp/Lerp_Position(
     A:NiagaraPosition=&Add_Vector(
      A=$System.SampleSkeletalMesh.SampledPosition
      B=&RotateVectorByQuaternion(Vector To Rotate=<-10 7 0>, Quaternion=$System.SampleSkeletalMesh.SampledRotation)
     )
     B:NiagaraPosition=&Lerp_Position(
      A:NiagaraPosition=Linked() {linkedVariable {name=System.SampleSkeletalMesh.SampledPosition, type=Vector3f}}
      B:NiagaraPosition=$System.SampleSkeletalMesh001.SampledPosition
      Alpha:NiagaraFloat=0.5
     )
     Alpha:NiagaraFloat=$System.Timeline.Float
    )) {isSetParameters=true}
   &ScaleColor(
    Scale Alpha=&FloatFromCurve(
     FloatCurve=DataInterface() {propertyValues {Curve {keys [{interpMode=RCIM_Cubic, tangentMode=RCTM_User, time=0, value=0}, <at 0.7 1>, {interpMode=RCIM_Cubic, tangentMode=RCTM_User, time=1, value=0}]}}}
     Scale Curve=&FloatFromCurve(
      FloatCurve=DataInterface() {propertyValues {Curve {keys [
          {interpMode=RCIM_Cubic, tangentMode=RCTM_User, time=0, value=0}
          <at 2.9 1>
          {interpMode=RCIM_Cubic, tangentMode=RCTM_User, time=3.025, value=0}
         ]}}}
      CurveIndex=$System.Age
     )
    )
    Scale RGB=&MakeVectorFromLinearColorRGB(Color=&ColorFromCurve(ColorCurve=DataInterface() {propertyValues {
        AlphaCurve {keys [<at 0 1 RCIM_Cubic>]}
        BlueCurve {keys [<at 0 1 RCIM_Cubic>,<at 0.5272 1.5>,<at 1 2 RCIM_Cubic>]}
        GreenCurve {keys [<at 0 0.4751 RCIM_Cubic>,<at 0.5272 1.3854>,<at 1 2 RCIM_Cubic>]}
        RedCurve {keys [<at 0 0.0831 RCIM_Cubic>,<at 0.5272 0.1924>,<at 1 2 RCIM_Cubic>]}
       }}))
    ScaleA=true
    ScaleRGB=true
   )
   &ScaleMeshSize(Scale Factor=&VectorFromFloat(Value=&FloatFromCurve(
      FloatCurve=DataInterface() {propertyValues {Curve {keys [{tangentMode=RCTM_User, time=0, value=0}, {tangentMode=RCTM_User, time=3.1, value=1}]}}}
      CurveIndex=$System.Age
     )))
   &DynamicMaterialParameters(
    Index 0 Param 2=1
    Index 0 Param 3=&FloatFromCurve(
     FloatCurve=DataInterface() {propertyValues {Curve {keys [{arriveTangent=-25.4246, time=0, value=1}, <at 1 0>]}}}
     Scale Curve=&FloatFromCurve(
      FloatCurve=DataInterface() {propertyValues {Curve {keys [{tangentMode=RCTM_User, time=0, value=0}, {tangentMode=RCTM_User, time=3.1, value=1}]}}}
      CurveIndex=$System.Age
     )
    )
    Index 0 Param 4=&FloatFromCurve(FloatCurve=DataInterface() {propertyValues {Curve {keys [<at 0 2>,<at 1 1>]}}})
    Module.Param0WriteEnabled=true
    Module.Param1WriteEnabled=true
    Module.Param2WriteEnabled=true
    Module.Param3WriteEnabled=true
   )
  ]
  rendererClasses [&NiagaraMeshRendererProperties]
  renderers [{rendererClass=&NiagaraMeshRendererProperties, rendererIndex=0}]
 }
 Light {
  emitterUpdateScript [
   &EmitterState()
   &SpawnBurst_Instantaneous(Spawn Count=1, Use Spawn Probability=false)
  ]
  particleSpawnScript [&InitializeParticle(
    Color=<rgba 0.401 2.86 7 1>
    Lifetime=3
    Sprite Size Mode=::NewEnumerator3
    Uniform Sprite Size=25
    UsePositionOffset=false
    Write Material Random=true
   )]
  particleUpdateScript [
   &ParticleState()
   /SetVariables(Particles.Position=&Lerp_Position(
     A:NiagaraPosition=&Add_Vector(
      A=$System.SampleSkeletalMesh.SampledPosition
      B=&RotateVectorByQuaternion(Vector To Rotate=<-10 7 0>, Quaternion=$System.SampleSkeletalMesh.SampledRotation)
     )
     B:NiagaraPosition=&Lerp_Position(
      A:NiagaraPosition=Linked() {linkedVariable {name=System.SampleSkeletalMesh.SampledPosition, type=Vector3f}}
      B:NiagaraPosition=$System.SampleSkeletalMesh001.SampledPosition
      Alpha:NiagaraFloat=0.5
     )
     Alpha:NiagaraFloat=$System.Timeline.Float
    )) {isSetParameters=true}
   &FN_Light_Attributes(
    Color=$Particles.Color
    Create Color Attribute=true
    Create Exponent Attribute=true
    Create Radius Attribute=true
    Exponent=6
    Radius=120
    Scale Color=&FloatFromCurve(FloatCurve=DataInterface() {propertyValues {Curve {keys [<at 0 0>,<at 1 1>]}}}, Scale Curve=5)
   )
   &ScaleSpriteSize(Scale Factor=&Vector2DFromFloat(Value=&FloatFromCurve(FloatCurve=DataInterface() {propertyValues {Curve {keys [
          {arriveTangent=4.3794, interpMode=RCIM_Cubic, leaveTangent=2.141, tangentMode=RCTM_User, time=0, value=0}
          {arriveTangent=0.353, interpMode=RCIM_Cubic, leaveTangent=0.353, tangentMode=RCTM_User, time=1, value=1}
         ]}}})))
   &ScaleColor(
    Scale Alpha=&FloatFromCurve(FloatCurve=DataInterface() {propertyValues {Curve {keys [
         {arriveTangent=2.3365, interpMode=RCIM_Cubic, leaveTangent=2.3365, tangentMode=RCTM_User, time=0, value=0}
         {interpMode=RCIM_Cubic, tangentMode=RCTM_User, time=0.5, value=1}
        ]}}})
    ScaleA=true
    ScaleRGB=false
   )
   &FN_Sprite_Attributes(
    Create SpriteSize Attribute=true
    SpriteSize=&Multiply_Vector2D_ByFloat(Float=2.5, Vector2D=$Particles.SpriteSize)
    Write to Module Namespace=true
   )
   &FN_Sprite_Attributes(
    Color=<rgba 0.0069 0 0.0139 1.35>
    Create Color Attribute=true
    Create SpriteSize Attribute=true
    SpriteSize=&Multiply_Vector2D_ByFloat(Float=4, Vector2D=$Particles.SpriteSize)
    Write to Module Namespace=true
   ) {module=FN_Sprite_Attributes001}
   &FN_Mesh_Attributes(
    Create Color Attribute=true
    Create Scale Attribute=true
    Scale=&VectorFromFloat(Value=&FloatFromCurve(
      FloatCurve=DataInterface() {propertyValues {Curve {keys [
          {arriveTangent=0.838, interpMode=RCIM_Cubic, leaveTangent=0.838, tangentMode=RCTM_User, time=0, value=0.7}
          {arriveTangent=0.353, interpMode=RCIM_Cubic, leaveTangent=0.353, tangentMode=RCTM_User, time=1, value=1}
         ]}}}
      Scale Curve=0.17
     ))
    Write to Module Namespace=true
   )
  ]
  rendererClasses [
   &NiagaraSpriteRendererProperties
   &NiagaraLightRendererProperties
   &NiagaraMeshRendererProperties
  ]
  renderers [
   {rendererClass=&NiagaraSpriteRendererProperties}
   {rendererClass=&NiagaraLightRendererProperties, rendererIndex=1}
   {rendererClass=&NiagaraMeshRendererProperties, rendererIndex=2}
   {rendererClass=&NiagaraSpriteRendererProperties, rendererIndex=3}
  ]
 }
 Magic {
  emitterUpdateScript [
   &EmitterState()
   &SpawnRate(SpawnRate=2, Use Spawn Probability=false)
  ]
  particleSpawnScript [/Niagara/Modules/Spawn/Initialization/InitializeParticle(
    Color=<rgba 1 1 1 0.4>
    Lifetime=0.8
    Module.Write Color=true
    Module.Write Lifetime=true
    Module.Write Mass=true
    Module.Write Material Random=true
    Module.Write Position=false
    Module.Write Scale=false
    Module.Write SpriteRotation=true
    Module.Write SpriteSize=true
    Position=&SimulationPosition()
    Sprite Rotation=&RandomRangeFloat(Minimum=-180, Maximum=180)
    Sprite Size=<300 200>
   )]
  particleUpdateScript [
   &ParticleState()
   &SkeletalMeshLocation(
    Bone Sampling Mode=::NewEnumerator4
    Inherit Velocity=false
    Interpolate Mesh Sampling=false
    MaskSpawnGroup=false
    Skeletal Mesh=$User.Hands
   ) {enabled=false}
   /SetVariables(Particles.Position=&Lerp_Position(
     A:NiagaraPosition=&Add_Vector(
      A=$System.SampleSkeletalMesh.SampledPosition
      B=&RotateVectorByQuaternion(Vector To Rotate=<-10 7 0>, Quaternion=$System.SampleSkeletalMesh.SampledRotation)
     )
     B:NiagaraPosition=&Lerp_Position(
      A:NiagaraPosition=Linked() {linkedVariable {name=System.SampleSkeletalMesh.SampledPosition, type=Vector3f}}
      B:NiagaraPosition=$System.SampleSkeletalMesh001.SampledPosition
      Alpha:NiagaraFloat=0.5
     )
     Alpha:NiagaraFloat=$System.Timeline.Float
    )) {isSetParameters=true}
   &ScaleSpriteSize(Scale Factor=&Vector2DFromFloat(Value=&FloatFromCurve(
      FloatCurve=DataInterface() {propertyValues {Curve {keys [
          {arriveTangent=-0.597, interpMode=RCIM_Cubic, leaveTangent=-0.597, tangentMode=RCTM_User, time=0.0023, value=1}
          {arriveTangent=-1.4856, interpMode=RCIM_Cubic, leaveTangent=-1.4856, tangentMode=RCTM_User, time=1, value=0.2}
         ]}}}
      Scale Curve=&FloatFromCurve(
       FloatCurve=DataInterface() {propertyValues {Curve {keys [
           {interpMode=RCIM_Cubic, tangentMode=RCTM_User, time=0, value=0}
           {interpMode=RCIM_Cubic, tangentMode=RCTM_User, time=3, value=1}
           {arriveTangent=-11.2792, interpMode=RCIM_Cubic, leaveTangent=-11.2792, tangentMode=RCTM_User, time=3.1, value=0}
          ]}}}
       CurveIndex=$System.Age
      )
     )))
   &ScaleColor(
    Scale Alpha=&FloatFromCurve(
     FloatCurve=DataInterface() {propertyValues {Curve {keys [
         {arriveTangent=2.2202, interpMode=RCIM_Cubic, leaveTangent=2.2202, tangentMode=RCTM_User, time=0.0023, value=0}
         {arriveTangent=0.8201, interpMode=RCIM_Cubic, leaveTangent=0.8201, tangentMode=RCTM_User, time=1, value=1}
        ]}}}
     Scale Curve=&FloatFromCurve(FloatCurve=DataInterface() {propertyValues {Curve {keys [<at 0 1>,<at 3 1>,<at 3.1 0>]}}}, CurveIndex=$Engine.System.Age)
    )
    Scale RGB=&MakeVectorFromLinearColorRGB(Color=&ColorFromCurve(ColorCurve=DataInterface() {propertyValues {
        AlphaCurve {keys [<at 0 1 RCIM_Cubic>]}
        BlueCurve {keys [<at 0.2139 1 RCIM_Cubic>,<at 1 1 RCIM_Cubic>]}
        GreenCurve {keys [<at 0.2139 0.6852 RCIM_Cubic>,<at 1 0.9339 RCIM_Cubic>]}
        RedCurve {keys [<at 0.2139 0.0608 RCIM_Cubic>,<at 1 0.8708 RCIM_Cubic>]}
       }}))
    ScaleA=true
    ScaleRGB=true
   )
   &FN_Radial_Material(
    Base Sprite Tile U=2
    Depth Fade=200
    Radial Mask Power=4
    Radial Mask Strength=1.5
    Sprite Extrusion Speed=-0.5
    Sprite Spin Speed:NiagaraFloat=&RandomRangeFloat(Minimum=-0.05, Maximum=0.05)
   )
  ]
  rendererClasses [&NiagaraSpriteRendererProperties]
  renderers [{rendererClass=&NiagaraSpriteRendererProperties}]
 }
 Rays {
  emitterUpdateScript [
   &EmitterState()
   &SpawnBurst_Instantaneous(Spawn Count=2, Use Spawn Probability=false)
  ]
  particleSpawnScript [
   &FN_SetSpawnGroup(Number of Spawn Groups=2)
   /Niagara/Modules/Spawn/Initialization/V2/InitializeParticle(
    Color=<rgba 2.0295 5.6166 7 0.5>
    Lifetime=3.05
    Mesh Scale Mode=::NewEnumerator3
    Mesh Uniform Scale=40
    Position Offset=<8 0 0>
    UsePositionOffset=true
    Write Material Random=true
   ) {module=InitializeParticle001}
   &InitialMeshOrientation(
    Rotation=&RandomRangeVector(Minimum=<-1 0 0>, Maximum=<1 0 0>)
    Use Orientation Vector=true
    Use Rotation Vector=true
   )
   &DynamicMaterialParameters(Module.Index0Float0Write=true, Module.Index0Float1Write=true, Module.Index0Float2Write=true, Module.Index0Float3Write=true)
  ]
  particleUpdateScript [
   &ParticleState()
   /SetVariables(Particles.Position=&Lerp_Position(
     A:NiagaraPosition=&Add_Vector(
      A=Linked() {linkedVariable {name=System.SampleSkeletalMesh.SampledPosition, type=NiagaraPosition}}
      B=&RotateVectorByQuaternion(Vector To Rotate=<-10 7 0>, Quaternion=$System.SampleSkeletalMesh.SampledRotation)
     )
     B:NiagaraPosition=&Lerp_Position(
      A:NiagaraPosition=$System.SampleSkeletalMesh.SampledPosition
      B:NiagaraPosition=$System.SampleSkeletalMesh001.SampledPosition
      Alpha:NiagaraFloat=0.5
     )
     Alpha:NiagaraFloat=$System.Timeline.Float
    )) {isSetParameters=true}
   &UpdateMeshOrientation(Rotation Rate=/Game/Effects/Niagara/DynamicInputs/FN_PickFloatFromSpawnGroupCutoff(
     Over Cutoff Float:NiagaraFloat=0.05
     Under Cutoff Float:NiagaraFloat=-0.05
     SpawnGroup:NiagaraInt32=$Particles.SpawnGroup
     Cutoff:NiagaraInt32=1
    ))
   &ScaleMeshSize(Scale Factor=&VectorFromFloat(Value=&FloatFromCurve(FloatCurve=DataInterface() {propertyValues {Curve {keys [{arriveTangent=4.3794, time=0, value=0}, {interpMode=RCIM_Cubic, leaveTangent=0.4, time=1, value=1}]}}})))
   /SetVariables(Particles.Renderer2_Color:LinearColor=&ColorFromCurve(ColorCurve=DataInterface() {propertyValues {
       AlphaCurve {keys [<at 0.0035 0 RCIM_Cubic>,<at 0.98 0.5 RCIM_Cubic>,<at 0.9969 0>]}
       BlueCurve {keys [<at 0 1 RCIM_Cubic>,<at 1 1 RCIM_Cubic>]}
       GreenCurve {keys [<at 0 1 RCIM_Cubic>,<at 1 1 RCIM_Cubic>]}
       RedCurve {keys [<at 0 1 RCIM_Cubic>,<at 1 1 RCIM_Cubic>]}
      }})) {enabled=false, isSetParameters=true}
   /SetVariables(Particles.Renderer2_Size:Vector3f=&Multiply_VectorByFloat(Vector=$Particles.Scale, Float=0.1)) {enabled=false, isSetParameters=true}
  ]
  rendererClasses [&NiagaraMeshRendererProperties]
  renderers [{rendererClass=&NiagaraMeshRendererProperties, rendererIndex=0}]
 }
 Rays_LWE {bEnabled=false, rendererClasses [], renderers []}
 Rings {
  emitterUpdateScript [
   &EmitterState()
   &SpawnRate(SpawnRate=8, Use Spawn Probability=false)
   &FN_Timer()
  ]
  particleSpawnScript [
   &InitializeParticle(
    Lifetime Max=0.6
    Lifetime Min=0.4
    Lifetime Mode=::NewEnumerator1
    Mesh Scale Mode=::NewEnumerator3
    Mesh Uniform Scale=5
    UsePositionOffset=false
    Write Material Random=true
   )
   &InitialMeshOrientation(
    Rotation=&RandomRangeVector(Minimum=<-1 0 0>, Maximum=<1 0 0>)
    Use Orientation Vector=true
    Use Rotation Vector=true
   )
  ]
  particleUpdateScript [
   &ParticleState()
   /SetVariables(Particles.Position=&Lerp_Position(
     A:NiagaraPosition=&Add_Vector(
      A=$System.SampleSkeletalMesh.SampledPosition
      B=&RotateVectorByQuaternion(Vector To Rotate=<-10 7 0>, Quaternion=$System.SampleSkeletalMesh.SampledRotation)
     )
     B:NiagaraPosition=&Lerp_Position(
      A:NiagaraPosition=Linked() {linkedVariable {name=System.SampleSkeletalMesh.SampledPosition, type=Vector3f}}
      B:NiagaraPosition=$System.SampleSkeletalMesh001.SampledPosition
      Alpha:NiagaraFloat=0.5
     )
     Alpha:NiagaraFloat=$System.Timeline.Float
    )) {isSetParameters=true}
   &DynamicMaterialParameters(
    Index 3 Param 1=&Add_Float(A=&RandomRangeFloat(Minimum=0, Maximum=1), B=$Emitter.FN_Timer.Value)
    Index 3 Param 2=&FloatFromCurve(
     FloatCurve=DataInterface() {propertyValues {Curve {keys [
         {interpMode=RCIM_Cubic, tangentMode=RCTM_User, time=0, value=0}
         {interpMode=RCIM_Cubic, tangentMode=RCTM_User, time=0.5, value=0.5}
         {tangentMode=RCTM_User, time=1, value=0.5}
        ]}}}
     Scale Curve=&FloatFromCurve(
      FloatCurve=DataInterface() {propertyValues {Curve {keys [{tangentMode=RCTM_User, time=0, value=0}, {tangentMode=RCTM_User, time=3.1, value=1}]}}}
      CurveIndex=$System.Age
     )
    )
    Module.Param0WriteEnabled3=true
    Module.Param1WriteEnabled3=true
    Module.Param2WriteEnabled3=true
    Module.Param3WriteEnabled3=true
    Write Parameter Index 0=false
    Write Parameter Index 3=true
   )
   &ScaleColor(
    Scale Alpha=&FloatFromCurve(
     FloatCurve=DataInterface() {propertyValues {Curve {keys [{interpMode=RCIM_Cubic, tangentMode=RCTM_User, time=0, value=0}, {interpMode=RCIM_Cubic, tangentMode=RCTM_User, time=1, value=1}]}}}
     Scale Curve=&FloatFromCurve(FloatCurve=DataInterface() {propertyValues {Curve {keys [<at 0 1>,<at 2.9 1>,<at 3.05 0>]}}}, CurveIndex=$Engine.System.Age)
    )
    Scale RGB=&VectorFromFloat(Value=&FloatFromCurve(FloatCurve=DataInterface() {propertyValues {Curve {keys [<at 0 1>,<at 1 2>]}}}))
    ScaleA=true
    ScaleRGB=true
   )
   &ScaleMeshSize(Scale Factor=&VectorFromCurve(
     VectorCurve=DataInterface() {propertyValues {
       XCurve {keys [
         {arriveTangent=2.2871, interpMode=RCIM_Cubic, time=0, value=1}
         {arriveTangent=-1, interpMode=RCIM_Cubic, leaveTangent=-1, time=0.5, value=0.3}
         <at 1 0 RCIM_Cubic>
        ]}
       YCurve {keys [
         {arriveTangent=2.2871, interpMode=RCIM_Cubic, time=0, value=1}
         {arriveTangent=-1, interpMode=RCIM_Cubic, leaveTangent=-1, time=0.5, value=0.3}
         <at 1 0 RCIM_Cubic>
        ]}
       ZCurve {keys [
         {arriveTangent=2.2871, interpMode=RCIM_Cubic, time=0, value=1}
         {arriveTangent=-1, interpMode=RCIM_Cubic, leaveTangent=-1, time=0.5, value=0.3}
         <at 1 0 RCIM_Cubic>
        ]}
      }}
     Scale Curve=&VectorFromFloat(Value=&FloatFromCurve(
       FloatCurve=DataInterface() {propertyValues {Curve {keys [
           {interpMode=RCIM_Cubic, tangentMode=RCTM_User, time=0, value=0.5}
           {interpMode=RCIM_Cubic, tangentMode=RCTM_User, time=3, value=1}
           {arriveTangent=-11.2792, interpMode=RCIM_Cubic, leaveTangent=-11.2792, tangentMode=RCTM_User, time=3.1, value=0}
          ]}}}
       CurveIndex=$System.Age
      ))
    ))
  ]
  rendererClasses [&NiagaraMeshRendererProperties]
  renderers [{rendererClass=&NiagaraMeshRendererProperties, rendererIndex=0}]
 }
}
summary {
 systemName=NS_Hawaii_Beam_Charge
 emitters [
  {emitterName=Magic, rendererClasses [&NiagaraSpriteRendererProperties]}
  {emitterName=Rings, rendererClasses [&NiagaraMeshRendererProperties]}
  {emitterName=Fireworks_Mesh, rendererClasses [&NiagaraMeshRendererProperties]}
  {
   emitterName=Light
   rendererClasses [
    &NiagaraSpriteRendererProperties
    &NiagaraLightRendererProperties
    &NiagaraMeshRendererProperties
   ]
  }
  {emitterName=Rays, rendererClasses [&NiagaraMeshRendererProperties]}
  {bEnabled=false, emitterName=Rays_LWE, rendererClasses []}
 ]
 userVariables [
  User.BeamEnd_0 {type=Vector3f, description="", defaultValue {struct=Vector3f, value=<100 0 0>}}
  User.BeamStart_0_0 {type=Vector3f, description="", defaultValue {struct=Vector3f, value=<0 0 0>}}
  User.EquipState {type=NiagaraInt32}
  User.Hands {
   type=NiagaraDataInterfaceSkeletalMesh
   defaultValue {
    struct=NiagaraExt_VariableValue_DataInterface
    value {dataInterface=&self:NiagaraDataInterfaceSkeletalMesh_0}
   }
  }
 ]
}
