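# Niagara aura effect for a prototype Hawaii stamina beam item: a single cyan, translucent flat mesh emitter, burst-spawned just below the system origin.
# Particles outlive their lifetime until the emitter deactivates, with alpha ramped by a curve.
# Each update dots a camera vector with world up into CameraVectorAlpha.
# That value drives four dynamic material parameters and mesh scale, so the aura's look adapts to viewing angle.
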NiagaraSystem /StaminaGameplay/Gameplay/Items/Hawaii/PROTOTYPE/FX/NS_StaminaBeam_Aura
dependencies {
 usedDataInterfaces [&NiagaraDataInterfaceCamera, &NiagaraDataInterfaceCurve]
 usedDynamicInputs []
 usedModules [&SystemState]
}
emitters {Aura_Mesh_Flat {
  emitterUpdateScript [
   &EmitterState(
    Inactive Response=::NewEnumerator1
    Life Cycle Mode=::NewEnumerator1
    Loop Behavior=::NewEnumerator1
    Loop Duration Mode=::NewEnumerator1
    UseLoopDelay=false
   )
   &SpawnBurst_Instantaneous(Use Spawn Probability=false, UseLoopCountLimit=false)
  ]
  particleSpawnScript [
   &InitializeParticle(
    Color=<rgba 0.3012 1.236 2 0.5>
    Lifetime=1.7
    Mesh Scale Mode=::NewEnumerator3
    Mesh Uniform Scale=4
    UsePositionOffset=false
    Write Material Random=true
   )
   &SystemLocation(MaskSpawnGroup=false, Offset=<0 0 -5>)
  ]
  particleUpdateScript [
   &ParticleState(Kill Particles When Lifetime Has Elapsed=false, Let Infinitely Lived Particles Die When Emitter Deactivates=true)
   /SetVariables(Particles.CameraVectorAlpha:NiagaraFloat=&DotProduct(Vector A=&ReturnCameraVector(Camera Property=::NewEnumerator2, Coordinate Space=::World), Vector B=<0 0 1>)) {isSetParameters=true}
   &ScaleColor(
    Scale Alpha=&FloatFromCurve(FloatCurve=DataInterface() {propertyValues {Curve {keys [{interpMode=RCIM_Cubic, tangentMode=RCTM_User, time=0, value=0}, {interpMode=RCIM_Cubic, tangentMode=RCTM_User, time=1, value=1}]}}})
    ScaleA=true
    ScaleRGB=true
   )
   &DynamicMaterialParameters(
    Index 0 Param 1=&Lerp_Float(A=0, B=0.15, Alpha=$Particles.CameraVectorAlpha)
    Index 0 Param 2=&Lerp_Float(A=1, B=175, Alpha=$Particles.CameraVectorAlpha)
    Index 0 Param 3=&Lerp_Float(A=0, B=-50, Alpha=$Particles.CameraVectorAlpha)
    Index 0 Param 4=&FloatFromCurve(
     FloatCurve=DataInterface() {propertyValues {Curve {keys [<at 0 1 RCIM_Cubic>, {arriveTangent=0.2662, interpMode=RCIM_Cubic, leaveTangent=0.2662, tangentMode=RCTM_User, time=1, value=0.15}]}}}
     CurveIndex=$Particles.CameraVectorAlpha
    )
    Module.Param0WriteEnabled=true
    Module.Param1WriteEnabled=true
    Module.Param2WriteEnabled=true
    Module.Param3WriteEnabled=true
   ) {module=DynamicMaterialParameters001}
   &ScaleMeshSize(Scale Factor=&VectorFromFloat(Value=&FloatFromCurve(
      FloatCurve=DataInterface() {propertyValues {Curve {keys [{interpMode=RCIM_Cubic, tangentMode=RCTM_User, time=0, value=1.5}, {interpMode=RCIM_Cubic, tangentMode=RCTM_User, time=1, value=1}]}}}
      CurveIndex=$Particles.CameraVectorAlpha
     )))
  ]
  rendererClasses [&NiagaraMeshRendererProperties]
  renderers [{rendererClass=&NiagaraMeshRendererProperties, rendererIndex=0}]
 }}
summary {
 systemName=NS_StaminaBeam_Aura
 emitters [{emitterName=Aura_Mesh_Flat, rendererClasses [&NiagaraMeshRendererProperties]}]
 userVariables []
}
