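# Niagara effect for the prototype Hawaii StaminaBeam item while equipped, drawing a hand-held glow, ready-to-fire particles and velocity-driven ribbon trails.
# User.EquipState indexes arrays that set glow size, color and light, and gates both ReadyToFire emitters to state 0.
# System timelines ramp the glow and scale ready-particle spawning, while a waveform adds pulsing.
# Emitters share System.NewPos, and Glow also drives persistent audio, passing it state and position.
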NiagaraSystem /StaminaGameplay/Gameplay/Items/Hawaii/PROTOTYPE/FX/NS_StaminaBeam_Equipped
dependencies {
 usedDataInterfaces [
  &NiagaraDataInterfaceArrayFloat
  &NiagaraDataInterfaceCurve
  &NiagaraDataInterfaceArrayColor
  &NiagaraDataInterfaceColorCurve
 ]
 usedDynamicInputs [
  &SetBoolByFloatComparison
  &MakeFloatFromInt
  &SetBoolByIntComparison
  /Niagara/DynamicInputs/Lerp/Lerp_Position
  &RotateVectorByQuaternion
 ]
 usedModules [
  &self:SystemSpawnScript.SystemScriptSource.SystemScriptGraph.NiagaraNodeAssignment_1.SetVariables_#1
  &SystemState
  &SampleSkeletalMesh
  &Timeline
  &self:SystemSpawnScript.SystemScriptSource.SystemScriptGraph.NiagaraNodeAssignment_0.SetVariables_#2
  /Game/Effects/Niagara/Modules/Logic/UpdateValueStateChange
 ]
}
emitters {
 FN_AttachedRibbon {
  emitterUpdateScript [
   &EmitterState()
   &FN_SpawnRibbon_v2(
    Spawn Mode=::NewEnumerator0
    SpawnRate:NiagaraFloat=&ClampFloat(
     Float=&Multiply_Float(A=&VectorLength(Vector=$Engine.Owner.Velocity, Fallback Length=0), B=0.075)
     Max=30
    )
   )
  ]
  particleSpawnScript [
   &InitializeParticle(
    Position:Vector3f=$System.NewPos
    Position Mode=::NewEnumerator1
    Sprite Size Mode=::NewEnumerator3
    UsePositionOffset=false
    Write Material Random=true
   ) {enabled=false}
   /Game/Effects/Niagara/Modules/Ribbons/FN_InitializeRibbon_v4(
    Custom Offset=true
    Enable Skeletal Position=true
    Local Vector Offset:Vector3f=<-4 10 -10>
    Offset Amount=1
    Ribbon Width=20
    SkeletalMesh:NiagaraDataInterfaceSkeletalMesh=$System.Hands_Only
   )
   &FN_FlipRibbonV()
  ]
  particleUpdateScript [
   &SolveForcesAndVelocity(Clamp Velocity=false, Limit Acceleration=false)
   /Game/Effects/Niagara/Modules/Ribbons/FN_UpdateRibbon_v4(Enable Skeletal Position=true, SkeletalMesh:NiagaraDataInterfaceSkeletalMesh=$System.Hands_Only)
   &DynamicMaterialParameters(
    Index 0 Param 1=&FloatFromCurve(
     FloatCurve=DataInterface() {propertyValues {Curve {keys [
         {interpMode=RCIM_Cubic, leaveTangent=-1.462, tangentMode=RCTM_Break, time=0.25, value=1}
         {arriveTangent=-0.7544, interpMode=RCIM_Cubic, tangentMode=RCTM_Break, time=1, value=-1}
        ]}}}
     Scale Curve=0.75
    )
    Index 0 Param 2=&FloatFromCurve(FloatCurve=DataInterface() {propertyValues {Curve {keys [<at 0 2>,<at 0.15 0>]}}})
    Index 0 Param 3=-1
    Index 0 Param 4=0.5
    Module.Param0WriteEnabled=true
    Module.Param1WriteEnabled=true
    Module.Param2WriteEnabled=true
    Module.Param3WriteEnabled=true
   )
   &ScaleColor(
    Scale Alpha=&FloatFromCurve(
     FloatCurve=DataInterface() {propertyValues {Curve {keys [
         {arriveTangent=0.2237, interpMode=RCIM_Cubic, leaveTangent=0.2237, tangentMode=RCTM_User, time=0, value=0}
         {interpMode=RCIM_Cubic, tangentMode=RCTM_User, time=0.1, value=1}
         {tangentMode=RCTM_User, time=1, value=0}
        ]}}}
     Scale Curve=0.25
    )
    Scale RGB=&MakeVectorFromLinearColorRGB(Color=&ColorFromCurve(ColorCurve=DataInterface() {propertyValues {
        AlphaCurve {keys [<at 0 1 RCIM_Cubic>]}
        BlueCurve {keys [<at 0 1 RCIM_Cubic>,<at 0.2352 1 RCIM_Cubic>]}
        GreenCurve {keys [<at 0 0.7459 RCIM_Cubic>, {interpMode=RCIM_Cubic, leaveTangent=-1.0881, time=0.2352, value=0.2656}]}
        RedCurve {keys [<at 0 0.2583 RCIM_Cubic>, {interpMode=RCIM_Cubic, leaveTangent=-0.667, time=0.2352, value=0.0302}]}
       }}))
    ScaleA=true
    ScaleRGB=true
   )
  ]
  rendererClasses [&NiagaraRibbonRendererProperties]
  renderers [{rendererClass=&NiagaraRibbonRendererProperties}]
 }
 Glow {
  emitterUpdateScript [
   &EmitterState(Life Cycle Mode=::NewEnumerator1, Loop Behavior=::NewEnumerator1, Loop Duration Mode=::NewEnumerator1, UseLoopDelay=false)
   &SpawnBurst_Instantaneous(Spawn Count=1, Use Spawn Probability=false)
  ]
  particleSpawnScript [&InitializeParticle(
    Color=<rgba 0.4896 1.0524 2 0.5>
    Lifetime=0.5
    Mesh Scale Mode=::NewEnumerator3
    Mesh Uniform Scale=0.11
    Sprite Size Mode=::NewEnumerator3
    Uniform Sprite Size=30
    UsePositionOffset=false
    Write Material Random=true
   ) {module=InitializeParticle001}]
  particleUpdateScript [
   &ParticleState(Kill Particles When Lifetime Has Elapsed=false, Let Infinitely Lived Particles Die When Emitter Deactivates=true)
   &SkeletalMeshLocation(
    Apply To Particle Position=false
    Bone Sampling Mode=::NewEnumerator4
    Inherit Velocity=false
    Interpolate Mesh Sampling=false
    MaskSpawnGroup=false
    Skeletal Mesh=$User.Hands
   ) {enabled=false}
   &SkeletalMeshLocation(
    Apply To Particle Position=false
    Bone / Socket Index=1
    Bone Sampling Mode=::NewEnumerator4
    Inherit Velocity=false
    Interpolate Mesh Sampling=false
    MaskSpawnGroup=false
    Skeletal Mesh=$User.Hands
   ) {enabled=false, module=SkeletalMeshLocation001}
   /SetVariables(Particles.Position=$System.NewPos) {isSetParameters=true}
   &ScaleMeshSize(Scale Factor=&VectorFromFloat(Value=&Add_Float(
      A=&SelectFloatFromArray(Direct Array Index=$User.EquipState, Float Selection Array=DataInterface() {propertyValues {FloatData [1,0,0,0.2]}})
      B=&ClampFloat(Float=&FloatFromCurve(
        FloatCurve=DataInterface() {propertyValues {Curve {keys [<at 0 0>, <at 0.99 0.5>, {arriveTangent=0.4977, leaveTangent=0.4977, time=1, value=1}]}}}
        CurveIndex=$System.Timeline002.Float
        Scale Curve=0.8
       ))
     )))
   /Niagara/Noise/Waveform(
    EnableExponent1=false
    EnableExponent2=false
    Global Amplitude Scale:NiagaraFloat=&Multiply_Float(A=&MakeFloatFromVector(Vector=$Particles.Scale, Channel=::NewEnumerator0), B=0.6)
    GlobalClampMax=false
    GlobalClampMin=false
    UseInput1=false
    UseInput3=false
    UsePhase1=false
    UsePhase2=true
    Waveform Count:ENiagara_WaveformCount=::NewEnumerator1
    "[1] Frequency"=0.85
    "[2] Frequency"=0.85
    "[2] Phase":NiagaraFloat=&Add_Float(A=$Emitter.Age, B=0.25)
   )
   &ScaleSpriteSize(
    Scale Sprite Size Mode=::NewEnumerator0
    Uniform Scale Factor=&Add_Float(
     A=&SelectFloatFromArray(Direct Array Index=$User.EquipState, Float Selection Array=DataInterface() {propertyValues {FloatData [1,1,0,0.5]}})
     B=&Add_Float(
      A=&ClampFloat(Float=&FloatFromCurve(
        FloatCurve=DataInterface() {propertyValues {Curve {keys [
            <at 0 0 RCIM_Cubic>
            {arriveTangent=1, interpMode=RCIM_Cubic, leaveTangent=1, time=0.99, value=0.5}
            {arriveTangent=3.8479, interpMode=RCIM_Cubic, leaveTangent=3.8479, tangentMode=RCTM_User, time=1, value=1}
           ]}}}
        CurveIndex=$System.Timeline002.Float
       ))
      B=&Multiply_Float(A=$Particles.Waveform.Waveform, B=1.5)
     )
    )
   )
   &ScaleMeshSize(Scale Factor=&Add_Vector(A=<1 1 1>, B=&VectorFromFloat(Value=$Particles.Waveform.Waveform))) {module=ScaleMeshSize001}
   &ScaleColor(
    Scale Alpha=&FloatFromCurve(FloatCurve=DataInterface() {propertyValues {Curve {keys [{interpMode=RCIM_Cubic, tangentMode=RCTM_User, time=0, value=0}, {interpMode=RCIM_Cubic, tangentMode=RCTM_User, time=1, value=1}]}}})
    Scale RGB=&Add_Vector(
     A=&MakeVectorFromLinearColorRGB(Color=&SelectLinearColorFromArray(
       Array Sampling Mode=::NewEnumerator1
       Direct Array Index=$User.EquipState
       Color Selection Array=DataInterface() {propertyValues {ColorData [<rgba 0.3765 1.545 2.5 1>, <rgba 0.3012 1.236 2 1>, <rgba 0 0 0 1>, <rgba 0 0.2251 0.6042 1>]}}
       Random Seed=?
      ))
     B=&VectorFromFloat(Value=&Add_Float(
       A=&Multiply_Float(A=$System.Timeline002.NormalizedTime, B=0.3)
       B=&Multiply_Float(A=$Particles.Waveform.Waveform, B=6)
      ))
    )
    ScaleA=true
    ScaleRGB=true
   )
   /SetVariables(Particles.Renderer2_Color:LinearColor=&MultiplyLinearColor(
     Linear Color=&ColorFromCurve(ColorCurve=DataInterface() {propertyValues {
        AlphaCurve {keys [<at 0.0035 0 RCIM_Cubic>,<at 0.9983 0.1 RCIM_Cubic>]}
        BlueCurve {keys [<at 0 1 RCIM_Cubic>,<at 1 1 RCIM_Cubic>]}
        GreenCurve {keys [<at 0 1 RCIM_Cubic>,<at 1 1 RCIM_Cubic>]}
        RedCurve {keys [<at 0 1 RCIM_Cubic>,<at 1 1 RCIM_Cubic>]}
       }})
     Scale Factor=&MakeVector4_FromFloat(Float=&Add_Float(
       A=&Add_Float(
        A=&SelectFloatFromArray(Direct Array Index=$User.EquipState, Float Selection Array=DataInterface() {propertyValues {FloatData [1.5,1,0,0.25]}})
        B=&Multiply_Float(A=$Particles.Waveform.Waveform, B=4)
       )
       B=&Multiply_Float(A=$System.Timeline002.NormalizedTime, B=1.25)
      ))
    )) {isSetParameters=true}
   /SetVariables(Particles.Renderer2_Size:Vector3f=&Multiply_VectorByFloat(Vector=$Particles.Scale, Float=2)) {isSetParameters=true}
   /SetVariables(Particles.Renderer3_COlor:LinearColor=&MultiplyLinearColor(Linear Color=$Particles.Color, Scale Factor=&MakeVector4_FromFloat(Float=0.6))) {isSetParameters=true}
   &FN_Light_Attributes(
    Color=$Particles.Color
    Create Color Attribute=true
    Create Exponent Attribute=true
    Create Radius Attribute=true
    Exponent=6
    Radius=&Add_Float(A=45, B=&Multiply_Float(A=$Particles.Waveform.Waveform, B=15))
    Scale Color=&Add_Float(
     A=&FloatFromCurve(
      FloatCurve=DataInterface() {propertyValues {Curve {keys [{interpMode=RCIM_Cubic, tangentMode=RCTM_User, time=0, value=0}, {interpMode=RCIM_Cubic, tangentMode=RCTM_User, time=1, value=1}]}}}
      Scale Curve=&Add_Float(
       A=&SelectFloatFromArray(Direct Array Index=$User.EquipState, Float Selection Array=DataInterface() {propertyValues {FloatData [1.5,0,0,0.5]}})
       B=&Multiply_Float(A=1, B=$System.Timeline002.NormalizedTime)
      )
     )
     B=&Multiply_Float(A=$Particles.Waveform.Waveform, B=3)
    )
   )
   /Niagara/Modules/Audio/PlayPersistentAudio(
    Audio Playback Position:NiagaraPosition=$System.NewPos
    Audio Player:NiagaraDataInterfaceAudioPlayer=$User.EquipAudioPlayer
    Enable Fade In / Out=true
    Fade In Time=0.25
    Play Audio:NiagaraBool=$System.DetectIntergerStateChange.Changed
   )
   /Niagara/Modules/Audio/UpdatePersistentAudio(
    Audio Player:NiagaraDataInterfaceAudioPlayer=$User.EquipAudioPlayer
    Int Value:NiagaraInt32=$User.EquipState
    Update Type:ENiagara_AudioParamType=::NewEnumerator1
   )
   &UpdatePersistentAudio(
    Audio Player:NiagaraDataInterfaceAudioPlayer=$User.EquipAudioPlayer
    Position:NiagaraPosition=$System.NewPos
    Update Type:ENiagara_AudioParamType=::NewEnumerator5
   ) {module=UpdatePersistentAudio001}
   &Color(Color=&SelectLinearColorFromArray(
     Array Sampling Mode=::NewEnumerator1
     Direct Array Index=$User.EquipState
     Color Selection Array=DataInterface() {propertyValues {ColorData [<rgba 0 1 0.1807 1>, <rgba 1 0 0.2636 1>, <rgba 1 0.7202 0 1>, <rgba 0 0.7206 1 1>]}}
     Random Seed=?
    )) {enabled=false}
  ]
  rendererClasses [
   &NiagaraSpriteRendererProperties
   &NiagaraLightRendererProperties
   &NiagaraMeshRendererProperties
  ]
  renderers [
   {rendererClass=&NiagaraSpriteRendererProperties}
   {rendererClass=&NiagaraLightRendererProperties, rendererIndex=1}
   {rendererClass=&NiagaraMeshRendererProperties, rendererIndex=2}
   {rendererClass=&NiagaraMeshRendererProperties, rendererIndex=3}
  ]
 }
 ReadyToFIre {
  emitterUpdateScript [
   &EmitterState()
   &SpawnRate(
    Spawn Probability=&SelectFloatFromArray(Direct Array Index=$User.EquipState, Float Selection Array=DataInterface() {propertyValues {FloatData [1,0,0,0]}})
    SpawnRate=&Multiply_Float(A=$System.Timeline001.InverseNormalizedTime, B=15)
    Use Spawn Probability=true
   )
  ]
  particleSpawnScript [&InitializeParticle(
    Color=<rgba 0.3012 1.236 2 1>
    Lifetime=0.8
    Mesh Scale Mode=::NewEnumerator3
    Mesh Uniform Scale=0.7
    Sprite Size Mode=::NewEnumerator3
    Uniform Sprite Size=45
    UsePositionOffset=false
    Write Material Random=true
   )]
  particleUpdateScript [
   &ParticleState()
   /SetVariables(Particles.Position=$System.NewPos) {isSetParameters=true, module=SetVariables_#3}
   &ScaleMeshSize(Scale Factor=&VectorFromFloat(Value=&FloatFromCurve(FloatCurve=DataInterface() {propertyValues {Curve {keys [
          {interpMode=RCIM_Cubic, tangentMode=RCTM_User, time=0, value=0}
          {interpMode=RCIM_Cubic, tangentMode=RCTM_User, time=0.5, value=1}
          {tangentMode=RCTM_User, time=1, value=0}
         ]}}})))
   &UpdateMeshOrientation(Rotation Rate=0.5)
   &FN_Mesh_Attributes(Color=<rgba 1 1 1 0.5>, Create Color Attribute=true, Write to Module Namespace=true)
   &FN_Radial_Material(Base Sprite Tile U=1.5, Sprite Extrusion Speed=1, Sprite Spin Speed=0.5, Write to Module Namespace=true)
   &ScaleSpriteSize(Uniform Curve Sprite Scale=DataInterface() {propertyValues {Curve {keys [
        {interpMode=RCIM_Cubic, tangentMode=RCTM_User, time=0, value=0}
        {interpMode=RCIM_Cubic, tangentMode=RCTM_User, time=0.5, value=1}
        {tangentMode=RCTM_User, time=1, value=0}
       ]}}})
   &FN_Sprite_Attributes(
    Color=<rgba 0 0.3726 1 1>
    Create Color Attribute=true
    Create SpriteSize Attribute=true
    SpriteSize=&Multiply_Vector2D_ByFloat(Float=2, Vector2D=$Particles.SpriteSize)
    Write to Module Namespace=true
   )
   &SpriteRotationRate(Rotation Rate=200)
  ]
  rendererClasses [&NiagaraMeshRendererProperties, &NiagaraSpriteRendererProperties]
  renderers [
   {rendererClass=&NiagaraMeshRendererProperties, rendererIndex=0}
   {rendererClass=&NiagaraSpriteRendererProperties, rendererIndex=1}
   {rendererClass=&NiagaraSpriteRendererProperties, rendererIndex=2}
  ]
 }
 ReadyToFIre002 {
  emitterUpdateScript [
   &EmitterState()
   &SpawnRate(
    Spawn Probability=&SelectFloatFromArray(Direct Array Index=$User.EquipState, Float Selection Array=DataInterface() {propertyValues {FloatData [1,0,0,0]}})
    SpawnRate=&Multiply_Float(A=$System.Timeline001.InverseNormalizedTime, B=75)
    Use Spawn Probability=true
   )
  ]
  particleSpawnScript [&InitializeParticle(
    Color=<rgba 0.3012 1.236 2 1>
    Lifetime=&RandomRangeFloat(Minimum=0.25, Maximum=0.5)
    Sprite Rotation Mode=::NewEnumerator2
    Sprite Size Mode=::NewEnumerator1
    Uniform Sprite Size Max=4
    Uniform Sprite Size Min=2
    UsePositionOffset=false
    Write Material Random=true
   )]
  particleUpdateScript [
   &ParticleState()
   /SetVariables(Particles.Position:Vector3f=$System.NewPos) {isSetParameters=true, module=SetVariables_#3}
   &ShapeLocation(
    MaskSpawnGroup=false
    Override Seed=false
    Shape Origin=$Particles.Position
    Sphere Radius=&Multiply_Float(
     A=&RandomRangeFloat(Minimum=5, Maximum=10)
     B=&FloatFromCurve(
      FloatCurve=DataInterface() {propertyValues {Curve {keys [
          {arriveTangent=1.7199, interpMode=RCIM_Cubic, leaveTangent=1.7199, tangentMode=RCTM_User, time=0, value=0}
          {interpMode=RCIM_Cubic, tangentMode=RCTM_User, time=1, value=1}
         ]}}}
      Scale Curve=4
     )
    )
    UseOwnerScale=true
   )
   &FN_Radial_Material(Base Sprite Tile U=1.5, Sprite Extrusion Speed=1, Sprite Spin Speed=0.5)
   &ScaleSpriteSize(Uniform Curve Sprite Scale=DataInterface() {propertyValues {Curve {keys [
        {interpMode=RCIM_Cubic, tangentMode=RCTM_User, time=0, value=0}
        {interpMode=RCIM_Cubic, tangentMode=RCTM_User, time=0.5, value=1}
        {tangentMode=RCTM_User, time=1, value=0}
       ]}}})
  ]
  rendererClasses [&NiagaraSpriteRendererProperties]
  renderers [{rendererClass=&NiagaraSpriteRendererProperties}]
 }
}
summary {
 systemName=NS_StaminaBeam_Equipped
 emitters [
  {
   emitterName=Glow
   rendererClasses [
    &NiagaraSpriteRendererProperties
    &NiagaraLightRendererProperties
    &NiagaraMeshRendererProperties
   ]
  }
  {emitterName=ReadyToFIre, rendererClasses [&NiagaraMeshRendererProperties, &NiagaraSpriteRendererProperties]}
  {emitterName=ReadyToFIre002, rendererClasses [&NiagaraSpriteRendererProperties]}
  {emitterName=FN_AttachedRibbon, rendererClasses [&NiagaraRibbonRendererProperties]}
 ]
 userVariables [
  User.EquipAudioPlayer {
   type=NiagaraDataInterfaceAudioPlayer
   description=""
   defaultValue {
    struct=NiagaraExt_VariableValue_DataInterface
    value {
     dataInterface=&self:NiagaraDataInterfaceAudioPlayer_0
     dataInterfaceClass=&NiagaraDataInterfaceAudioPlayer
    }
   }
  }
  User.EquipState {type=NiagaraInt32}
  User.RibbonIDOffset {type=NiagaraInt32}
  User.SkeletalMesh {
   type=NiagaraDataInterfaceSkeletalMesh
   defaultValue {
    struct=NiagaraExt_VariableValue_DataInterface
    value {dataInterface=&self:NiagaraDataInterfaceSkeletalMesh_5}
   }
  }
 ]
}
