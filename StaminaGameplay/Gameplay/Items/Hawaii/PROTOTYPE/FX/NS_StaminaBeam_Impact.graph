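# Prototype Niagara one-shot impact effect for the Stamina beam item.
# An expanding blue sphere with a light and a main impact mesh form the flash, while center and torus-ring mesh smoke cools from red to grey.
# GPU sprite debris bursts outward and falls under gravity, with collision disabled.
# User.Scale drives much of the sizing, and System.DistanceToCamera adds size and speed to keep the impact readable at range.
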NiagaraSystem /StaminaGameplay/Gameplay/Items/Hawaii/PROTOTYPE/FX/NS_StaminaBeam_Impact
dependencies {
 usedDataInterfaces [&NiagaraDataInterfaceColorCurve, &NiagaraDataInterfaceCurve, &NiagaraDataInterfaceCamera]
 usedDynamicInputs [
  &SimulationPosition
  &RemapRange
  &VectorLength
  &Subtract_Vector
  &FN_LOD_Distance_Remap
 ]
 usedModules [
  &self:SystemSpawnScript.SystemScriptSource.SystemScriptGraph.NiagaraNodeAssignment_1.SetVariables_#1
  &SystemState
  &self:SystemSpawnScript.SystemScriptSource.SystemScriptGraph.NiagaraNodeAssignment_0.SetVariables_#2
  &self:SystemSpawnScript.SystemScriptSource.SystemScriptGraph.NiagaraNodeAssignment_2.SetVariables_#3
  &self:SystemSpawnScript.SystemScriptSource.SystemScriptGraph.NiagaraNodeAssignment_3.SetVariables_#4
 ]
}
emitters {
 Debris {
  simTarget=GPUComputeSim
  emitterUpdateScript [
   &EmitterState()
   &SpawnBurst_Instantaneous(Spawn Count=&UniformRangedInt(Minimum=15, Maximum=24), Use Spawn Probability=false)
  ]
  particleSpawnScript [
   &InitializeParticle(
    Color=<rgba 0 0 0 0.8>
    Lifetime=&UniformRangedFloat(Minimum=2, Maximum=3.5)
    Sprite Size Mode=::NewEnumerator3
    Uniform Sprite Size=&MakeFloatFromVector2d(Vector 2D=&Vector2DFromFloat(Value=&Add_Float(
       A=&Multiply_Float(A=&UniformRangedFloat(Minimum=4, Maximum=30), B=$User.Scale)
       B=&Multiply_Float(A=$System.DistanceToCamera, B=200)
      )))
    UsePositionOffset=false
    Write Material Random=true
   ) {module=InitializeParticle001}
   /Niagara/Modules/Update/SubUV/SubUVAnimation(SubUV Animation Mode=::NewEnumerator1)
   /Niagara/Modules/Spawn/Location/SphereLocation(
    Hemisphere Z=true
    MaskSpawnGroup=false
    Module.Override Randomness=false
    Sphere Radius=&Multiply_Float(A=50, B=$User.Scale)
    Surface Only=false
   )
   &AddVelocityFromPoint(
    Module.Use Falloff Distance=false
    Velocity Strength=&Multiply_Float(
     A=&Add_Float(
      A=&RandomRangeFloat(Minimum=400, Maximum=700)
      B=&Multiply_Float(A=$System.DistanceToCamera, B=1700)
     )
     B=$User.Scale
    )
   )
   &AddVelocityInCone(
    Cone Angle=18
    Module.Use Velocity Falloff On Cone Axis=false
    Velocity Strength=&UniformRangedFloat(Minimum=200, Maximum=1250)
   )
  ]
  particleUpdateScript [
   &ParticleState()
   &AccelerationForce(Acceleration=<0 0 -2800>)
   &Collision(
    Age Colliding Particles=false
    Collision Enabled=&SetBoolByFloatComparison(A=$Particles.Age, B=0.3)
    Kill Occluded Particles=false
    Kill Offscreen Particles=true
    Module.Control Rest State Via Surface Normal=true
    Module.EnableMaxCollisionCount=false
    Module.Randomize Collision Normal Vector=true
    Particle Radius Scale=2
    Rest State Time Range=0.1
    Restitution=0.63
   ) {enabled=false}
   &SolveForcesAndVelocity(Clamp Velocity=false, Limit Acceleration=false)
   &SpriteRotationRate(Rotation Rate=&AngleConversion(Angle:NiagaraFloat=0, Angle Input:ENiagara_AngleInput=::NewEnumerator0, Angle Output:ENiagara_AngleInput=::NewEnumerator1))
   &ScaleSpriteSize(Scale Factor=&Vector2DFromFloat(Value=&FloatFromCurve(FloatCurve=DataInterface() {propertyValues {Curve {keys [<at 0.6918 1>,<at 1 0>]}}})))
  ]
  rendererClasses [&NiagaraSpriteRendererProperties]
  renderers [{rendererClass=&NiagaraSpriteRendererProperties}]
 }
 Impact_Main {
  emitterUpdateScript [
   &EmitterState()
   &SpawnBurst_Instantaneous(Spawn Count=1, Use Spawn Probability=false)
   &SpawnBurst_Instantaneous(Spawn Count=1, Spawn Time=0.1, Use Spawn Probability=false) {enabled=false, module=SpawnBurst_Instantaneous001}
  ]
  particleSpawnScript [
   &InitializeParticle(
    Lifetime=0.5
    Mesh Scale Mode=::NewEnumerator3
    Mesh Uniform Scale=&Add_Float(
     A=&RandomRangeFloat(Minimum=1.7, Maximum=2)
     B=&Multiply_Float(A=$System.DistanceToCamera, B=8)
    )
    UsePositionOffset=false
    Write Material Random=true
   ) {module=InitializeParticle001}
   /SetVariables(Particles.R2_Size:Vector3f=&VectorFromFloat(Value=&Add_Float(A=0.6, B=&Multiply_Float(A=$System.DistanceToCamera, B=2)))) {isSetParameters=true}
   &InitialMeshOrientation(Use Rotation Vector=false)
  ]
  particleUpdateScript [
   &ParticleState()
   &ScaleMeshSize(Scale Factor=&VectorFromFloat(Value=&FloatFromCurve(FloatCurve=DataInterface() {propertyValues {Curve {keys [
          {arriveTangent=9.901, interpMode=RCIM_Cubic, leaveTangent=9.901, tangentMode=RCTM_User, time=0, value=0.4}
          <at 0.8 7>
          {arriveTangent=-16.5746, interpMode=RCIM_Cubic, leaveTangent=-16.5746, tangentMode=RCTM_User, time=1, value=4}
         ]}}})))
   &ScaleColor(
    Scale Alpha=&FloatFromCurve(FloatCurve=DataInterface() {propertyValues {Curve {keys [<at 0.3 1 RCIM_Cubic>, {arriveTangent=-3.0307, interpMode=RCIM_Cubic, leaveTangent=-3.0307, tangentMode=RCTM_User, time=1, value=0}]}}})
    ScaleA=true
    ScaleRGB=true
   )
   /SetVariables(Particles.R2_DynMat:Vector4f=&MakeVector4(
     X=0
     Y=1
     Z=&FloatFromCurve(FloatCurve=DataInterface() {propertyValues {Curve {keys [{arriveTangent=-25.4246, time=0, value=0}, <at 1 1>]}}})
     W=&FloatFromCurve(
      FloatCurve=DataInterface() {propertyValues {Curve {keys [<at 0 0.5>,<at 1 1>]}}}
      Scale Curve=&Add_Float(A=1, B=&Multiply_Float(A=$System.DistanceToCamera, B=15))
     )
    )) {isSetParameters=true}
   &FN_Mesh_Attributes(
    Color=/Niagara/DynamicInputs/TypeConversions/MakeLinearColorFromFloat(Value:NiagaraFloat=&FloatFromCurve(FloatCurve=DataInterface() {propertyValues {Curve {keys [<at 0 1>,<at 1 0>]}}}))
    Create Color Attribute=true
    Create Scale Attribute=true
    Scale=&VectorFromFloat(Value=&Multiply_Float(
      A=&MakeFloatFromVector(Vector=$Particles.R2_Size, Channel=::NewEnumerator0)
      B=&FloatFromCurve(FloatCurve=DataInterface() {propertyValues {Curve {keys [{arriveTangent=2.9152, interpMode=RCIM_Cubic, leaveTangent=2.9152, tangentMode=RCTM_User, time=0, value=0}, <at 1 1 RCIM_Cubic>]}}})
     ))
    Write to Module Namespace=true
   )
  ]
  rendererClasses [&NiagaraMeshRendererProperties]
  renderers [
   {rendererClass=&NiagaraMeshRendererProperties, rendererIndex=0}
   {rendererClass=&NiagaraMeshRendererProperties, rendererIndex=1}
  ]
 }
 Smoke_Center {
  emitterUpdateScript [
   &EmitterState()
   &SpawnBurst_Instantaneous(Spawn Count=3, Use Spawn Probability=false)
  ]
  particleSpawnScript [
   &InitializeParticle(
    Lifetime=&UniformRangedFloat(Minimum=0.7, Maximum=0.9)
    Mesh Scale Mode=::NewEnumerator3
    Mesh Uniform Scale=&MakeFloatFromVector(
     Vector=&VectorFromFloat(Value=&Add_Float(
       A=&Multiply_Float(A=&UniformRangedFloat(Minimum=0.5, Maximum=0.6), B=$User.Scale)
       B=&Multiply_Float(A=$System.DistanceToCamera, B=2.3)
      ))
     Channel=::NewEnumerator0
    )
    UsePositionOffset=false
    Write Material Random=true
   ) {module=InitializeParticle001}
   &DynamicMaterialParameters(
    Index 0 Param 1=&Add_Float(
     A=&UniformRangedFloat(Minimum=1.5, Maximum=1.8)
     B=&Multiply_Float(A=$System.DistanceToCamera, B=9)
    )
    Index 0 Param 3=-0.4
    Module.Param0WriteEnabled=true
    Module.Param1WriteEnabled=false
    Module.Param2WriteEnabled=true
    Module.Param3WriteEnabled=false
   )
   &AddVelocity(
    Cone Angle=30
    Module.UseConeFalloff=false
    Module.UseVelDistribution=false
    Override Seed=false
    Velocity Mode=::NewEnumerator2
    Velocity Speed=&Add_Float(
     A=&RandomRangeFloat(Minimum=1000, Maximum=2000)
     B=&Multiply_Float(A=$System.DistanceToCamera, B=5000)
    )
   )
   &InitialMeshOrientation(Rotation=&UniformRangedVector(Minimum=<-1 -1 -1>, Maximum=<1 1 1>), Use Rotation Vector=true) {module=InitialMeshOrientation001}
  ]
  particleUpdateScript [
   &ParticleState()
   &AccelerationForce(Acceleration=&UniformRangedVector(Minimum=<0 0 600>, Maximum=<0 0 2000>))
   &Drag(Drag=8, Use Linear Drag=true, Use Rotational Drag=false)
   &SolveForcesAndVelocity(Clamp Velocity=false, Limit Acceleration=false)
   &ScaleColor(
    Scale Alpha=&FloatFromCurve(FloatCurve=DataInterface() {propertyValues {Curve {keys [
         <at 0 0.85 RCIM_Cubic>
         {arriveTangent=-0.43, interpMode=RCIM_Cubic, leaveTangent=-0.43, time=0.2, value=0.83}
         {arriveTangent=-1.0375, interpMode=RCIM_Cubic, leaveTangent=-1.0375, time=0.5, value=0.635}
         <at 1 0 RCIM_Cubic>
        ]}}})
    Scale RGB=&MakeVectorFromLinearColorRGB(Color=&ColorFromCurve(ColorCurve=DataInterface() {propertyValues {
        AlphaCurve {keys []}
        BlueCurve {keys [<at 0 0.0026 RCIM_Cubic>,<at 0.413 0.015>,<at 0.6199 0.0187>,<at 1 0.1091 RCIM_Cubic>]}
        GreenCurve {keys [<at 0 0.0011 RCIM_Cubic>,<at 0.413 0.0135>,<at 0.6199 0.0171>,<at 1 0.1076 RCIM_Cubic>]}
        RedCurve {keys [<at 0 1.7597 RCIM_Cubic>,<at 0.413 0.0294>,<at 0.6199 0.0289>,<at 1 0.1115 RCIM_Cubic>]}
       }}))
    ScaleA=true
    ScaleRGB=true
   ) {enabled=false}
   &ScaleColor(
    Scale Alpha=&FloatFromCurve(FloatCurve=DataInterface() {propertyValues {Curve {keys [
         {arriveTangent=-1.8332, interpMode=RCIM_Cubic, time=0, value=0.874}
         {arriveTangent=-0.874, interpMode=RCIM_Cubic, leaveTangent=-0.874, tangentMode=RCTM_Break, time=0.5574, value=0.6351}
         {arriveTangent=-2.2319, interpMode=RCIM_Cubic, leaveTangent=-3.4277, tangentMode=RCTM_User, time=0.9988, value=-0.5}
        ]}}})
    Scale RGB=&MakeVectorFromLinearColorRGB(Color=&ColorFromCurve(ColorCurve=DataInterface() {propertyValues {
        AlphaCurve {keys []}
        BlueCurve {keys [<at 0 0.0026 RCIM_Cubic>,<at 0.2422 0.015>,<at 0.3364 0.0187>,<at 1 0.1091 RCIM_Cubic>]}
        GreenCurve {keys [<at 0 0.0011 RCIM_Cubic>,<at 0.2422 0.0135>,<at 0.3364 0.0171>,<at 1 0.1076 RCIM_Cubic>]}
        RedCurve {keys [<at 0 1.7597 RCIM_Cubic>,<at 0.2422 0.0294>,<at 0.3364 0.0289>,<at 1 0.1115 RCIM_Cubic>]}
       }}))
    ScaleA=true
    ScaleRGB=true
   ) {module=ScaleColor001}
   &DynamicMaterialParameters(
    Index 0 Param 1=&UniformRangedFloat(Minimum=1, Maximum=1.8)
    Index 0 Param 2=&FloatFromCurve(FloatCurve=DataInterface() {propertyValues {Curve {keys [<at 0.13 0.77 RCIM_Cubic>, {interpMode=RCIM_Cubic, leaveTangent=-12.3793, time=1, value=-2}]}}})
    Index 0 Param 3=-0.4
    Index 0 Param 4=&FloatFromCurve(FloatCurve=DataInterface() {propertyValues {Curve {keys [<at 0 2.4 RCIM_Cubic>,<at 1 0.5 RCIM_Cubic>]}}}, Scale Curve=2)
    Module.Param0WriteEnabled=false
    Module.Param1WriteEnabled=true
    Module.Param2WriteEnabled=false
    Module.Param3WriteEnabled=true
   ) {module=DynamicMaterialParameters001}
   &ScaleMeshSize(Scale Factor=&VectorFromFloat(Value=&FloatFromCurve(FloatCurve=DataInterface() {propertyValues {Curve {keys [
          {arriveTangent=5.6759, interpMode=RCIM_Cubic, leaveTangent=5.6759, tangentMode=RCTM_User, time=0, value=0}
          {arriveTangent=2, interpMode=RCIM_Cubic, leaveTangent=2, time=0.2218, value=0.6656}
          <at 0.5 1>
          {arriveTangent=-1.9082, interpMode=RCIM_Cubic, leaveTangent=-1.9082, tangentMode=RCTM_User, time=1, value=0.35}
         ]}}})))
   &UpdateMeshOrientation(
    Rotation Rate=&Lerp_Float(A=-1, B=1, Alpha=$Particles.Initial.MaterialRandom)
    Rotation Vector=&Lerp_Vector(A=<-1 0 0>, B=<1 0 1>, Alpha=$Particles.Initial.MaterialRandom)
   ) {module=UpdateMeshOrientation001}
  ]
  rendererClasses [&NiagaraMeshRendererProperties]
  renderers [{rendererClass=&NiagaraMeshRendererProperties, rendererIndex=0}]
 }
 Smoke_Ring {
  emitterUpdateScript [
   &EmitterState()
   &SpawnBurst_Instantaneous(Spawn Count=7, Use Spawn Probability=false)
  ]
  particleSpawnScript [
   &InitializeParticle(
    Lifetime=&UniformRangedFloat(Minimum=0.6, Maximum=0.8)
    Mesh Scale Mode=::NewEnumerator3
    Mesh Uniform Scale=&MakeFloatFromVector(
     Vector=&VectorFromFloat(Value=&Add_Float(
       A=&Multiply_Float(A=&UniformRangedFloat(Minimum=0.3, Maximum=0.4), B=$User.Scale)
       B=&Multiply_Float(A=$System.DistanceToCamera, B=1.7)
      ))
     Channel=::NewEnumerator0
    )
    UsePositionOffset=false
    Write Material Random=true
   ) {module=InitializeParticle001}
   &DynamicMaterialParameters(
    Index 0 Param 1=&Add_Float(
     A=&UniformRangedFloat(Minimum=1.5, Maximum=1.8)
     B=&Multiply_Float(A=$System.DistanceToCamera, B=6)
    )
    Index 0 Param 3=-0.4
    Module.Param0WriteEnabled=true
    Module.Param1WriteEnabled=false
    Module.Param2WriteEnabled=true
    Module.Param3WriteEnabled=false
   )
   &InitialMeshOrientation(Rotation=&UniformRangedVector(Minimum=<-1 -1 -1>, Maximum=<1 1 1>), Use Rotation Vector=true) {module=InitialMeshOrientation001}
   &TorusLocation(
    Align Torus With Axis=true
    Large Radius:NiagaraFloat=&Multiply_Float(
     A=&Add_Float(A=12, B=&Multiply_Float(A=$System.DistanceToCamera, B=100))
     B=$User.Scale
    )
    MaskSpawnGroup=false
    Torus Axis=<1 0 0>
    Torus Distribution Mode:ENiagaraTorusDistributionMode=::NewEnumerator1
    Torus Mode=::NewEnumerator2
    U Position:NiagaraFloat=&ReturnNormalizedExecIndex(Normalized Index Includes One=false)
   ) {module=TorusLocation001}
   &AddVelocityFromPoint(
    Module.Use Falloff Distance=false
    Origin Offset Coordinate Space=::Simulation
    Velocity Strength=&Add_Float(
     A=&Multiply_Float(A=&RandomRangeFloat(Minimum=800, Maximum=1500), B=$User.Scale)
     B=&Multiply_Float(A=$System.DistanceToCamera, B=7000)
    )
   )
   &CameraOffset()
  ]
  particleUpdateScript [
   &ParticleState()
   &AccelerationForce(Acceleration=&UniformRangedVector(Minimum=<0 0 600>, Maximum=<0 0 2000>))
   &Drag(Drag=8, Use Linear Drag=true, Use Rotational Drag=false)
   &SolveForcesAndVelocity(Clamp Velocity=false, Limit Acceleration=false)
   &ScaleColor(
    Scale Alpha=&FloatFromCurve(FloatCurve=DataInterface() {propertyValues {Curve {keys [
         {arriveTangent=-1.8332, interpMode=RCIM_Cubic, time=0, value=0.874}
         {arriveTangent=-0.874, interpMode=RCIM_Cubic, leaveTangent=-0.874, tangentMode=RCTM_Break, time=0.5574, value=0.6351}
         {arriveTangent=-2.2319, interpMode=RCIM_Cubic, leaveTangent=-3.4277, tangentMode=RCTM_User, time=0.9988, value=-0.5}
        ]}}})
    Scale RGB=&MakeVectorFromLinearColorRGB(Color=&ColorFromCurve(ColorCurve=DataInterface() {propertyValues {
        AlphaCurve {keys []}
        BlueCurve {keys [<at 0 0.0026 RCIM_Cubic>,<at 0.2422 0.015>,<at 0.3364 0.0187>,<at 1 0.1091 RCIM_Cubic>]}
        GreenCurve {keys [<at 0 0.0011 RCIM_Cubic>,<at 0.2422 0.0135>,<at 0.3364 0.0171>,<at 1 0.1076 RCIM_Cubic>]}
        RedCurve {keys [<at 0 1.7597 RCIM_Cubic>,<at 0.2422 0.0294>,<at 0.3364 0.0289>,<at 1 0.1115 RCIM_Cubic>]}
       }}))
    ScaleA=true
    ScaleRGB=true
   ) {module=ScaleColor001}
   &DynamicMaterialParameters(
    Index 0 Param 1=&UniformRangedFloat(Minimum=1, Maximum=1.8)
    Index 0 Param 2=&FloatFromCurve(FloatCurve=DataInterface() {propertyValues {Curve {keys [<at 0.13 0.77 RCIM_Cubic>, {interpMode=RCIM_Cubic, leaveTangent=-12.3793, time=1, value=-2}]}}})
    Index 0 Param 3=-0.4
    Index 0 Param 4=&FloatFromCurve(FloatCurve=DataInterface() {propertyValues {Curve {keys [<at 0 2.4 RCIM_Cubic>,<at 1 0.5 RCIM_Cubic>]}}}, Scale Curve=2)
    Module.Param0WriteEnabled=false
    Module.Param1WriteEnabled=true
    Module.Param2WriteEnabled=false
    Module.Param3WriteEnabled=true
   ) {module=DynamicMaterialParameters001}
   &ScaleMeshSize(Scale Factor=&VectorFromFloat(Value=&FloatFromCurve(FloatCurve=DataInterface() {propertyValues {Curve {keys [
          {arriveTangent=5.6759, interpMode=RCIM_Cubic, leaveTangent=5.6759, tangentMode=RCTM_User, time=0, value=0}
          {arriveTangent=2, interpMode=RCIM_Cubic, leaveTangent=2, time=0.2218, value=0.6656}
          <at 0.5 1>
          {arriveTangent=-1.9082, interpMode=RCIM_Cubic, leaveTangent=-1.9082, tangentMode=RCTM_User, time=1, value=0.35}
         ]}}})))
   &UpdateMeshOrientation(
    Rotation Rate=&Lerp_Float(A=-1, B=1, Alpha=$Particles.Initial.MaterialRandom)
    Rotation Vector=&Lerp_Vector(A=<-1 0 0>, B=<1 0 1>, Alpha=$Particles.Initial.MaterialRandom)
   ) {module=UpdateMeshOrientation001}
  ]
  rendererClasses [&NiagaraMeshRendererProperties]
  renderers [{rendererClass=&NiagaraMeshRendererProperties, rendererIndex=0}]
 }
 initSphere001 {
  emitterUpdateScript [
   &EmitterState()
   &SpawnBurst_Instantaneous(Spawn Count=1, Use Spawn Probability=false)
  ]
  particleSpawnScript [
   &InitializeParticle(
    Color=<rgba 0 0.7453 2 1>
    Lifetime=0.35
    Mesh Scale Mode=::NewEnumerator3
    Mesh Uniform Scale=&MakeFloatFromVector(
     Vector=&VectorFromFloat(Value=&Add_Float(
       A=&Multiply_Float(A=16, B=$User.Scale)
       B=&Multiply_Float(A=$System.DistanceToCamera, B=30)
      ))
     Channel=::NewEnumerator0
    )
    UsePositionOffset=false
    Write Material Random=true
   ) {module=InitializeParticle001}
   &InitialMeshOrientation(Use Rotation Vector=false) {module=InitialMeshOrientation001}
   &DynamicMaterialParameters(
    Index 0 Param 1=&UniformRangedFloat(Minimum=0, Maximum=20)
    Module.Param0WriteEnabled=true
    Module.Param1WriteEnabled=false
    Module.Param2WriteEnabled=false
    Module.Param3WriteEnabled=false
   )
  ]
  particleUpdateScript [
   &ParticleState()
   /SetVariables(Particles.DistanceToCamera:NiagaraFloat=&NormalizeDistanceRange(
     Start Position:NiagaraPosition=$Particles.Position
     End Position:NiagaraPosition=&ReturnCameraVector(Camera Property=::NewEnumerator0, Coordinate Space=::World)
     Distance:NiagaraFloat=3000
     Invert Normalized Range:NiagaraBool=false
    )) {enabled=false, isSetParameters=true}
   &ScaleMeshSize(Scale Factor=&VectorFromFloat(Value=&FloatFromCurve(FloatCurve=DataInterface() {propertyValues {Curve {keys [{arriveTangent=5.2157, interpMode=RCIM_Cubic, leaveTangent=5.2157, tangentMode=RCTM_User, time=0, value=0}, <at 0.4 1 RCIM_Cubic>]}}})))
   &ScaleColor(
    Scale RGB=&VectorFromFloat(Value=&FloatFromCurve(FloatCurve=DataInterface() {propertyValues {Curve {keys [<at 0 0.7>,<at 0.7729 0.7>,<at 1 0>]}}}))
    ScaleA=true
    ScaleRGB=true
   )
   &FN_Mesh_Attributes(
    Color=&MultiplyLinearColor(
     Linear Color=<rgba 0 1.8631 5 1>
     Scale Factor=&MakeVector4(
      X=1
      Y=1
      Z=1
      W=&FloatFromCurve(FloatCurve=DataInterface() {propertyValues {Curve {keys [<at 0.4 1 RCIM_Cubic>,<at 1 0>]}}})
     )
    )
    Create Color Attribute=true
    Create Scale Attribute=true
    Scale=&VectorFromFloat(Value=&FloatFromCurve(
      FloatCurve=DataInterface() {propertyValues {Curve {keys [{arriveTangent=9.901, interpMode=RCIM_Cubic, leaveTangent=9.901, tangentMode=RCTM_User, time=0, value=0.4}, <at 1 6 RCIM_Cubic>]}}}
      Scale Curve=&Multiply_Float(A=$User.Scale, B=0.5)
     ))
    Write to Module Namespace=true
   )
   &FN_Light_Attributes(
    Color=$Particles.Color
    Create Color Attribute=true
    Create Exponent Attribute=true
    Create Position Attribute=true
    Create Radius Attribute=true
    Exponent=7
    Position=&AddVectorToPosition(Position=$Particles.Position, Vector=<300 0 0>)
    Radius=1000
    Scale Color=50
   )
  ]
  rendererClasses [&NiagaraMeshRendererProperties, &NiagaraLightRendererProperties]
  renderers [
   {rendererClass=&NiagaraMeshRendererProperties, rendererIndex=0}
   {rendererClass=&NiagaraMeshRendererProperties, rendererIndex=1}
   {rendererClass=&NiagaraLightRendererProperties, rendererIndex=2}
  ]
 }
}
summary {
 systemName=NS_StaminaBeam_Impact
 emitters [
  {emitterName=Smoke_Ring, rendererClasses [&NiagaraMeshRendererProperties]}
  {emitterName=Debris, simTarget=GPUComputeSim, rendererClasses [&NiagaraSpriteRendererProperties]}
  {emitterName=initSphere001, rendererClasses [&NiagaraMeshRendererProperties, &NiagaraLightRendererProperties]}
  {emitterName=Impact_Main, rendererClasses [&NiagaraMeshRendererProperties]}
  {emitterName=Smoke_Center, rendererClasses [&NiagaraMeshRendererProperties]}
 ]
 userVariables [
  User.ExplosionScale {type=NiagaraFloat, defaultValue=0, description=""}
  User.GridPivot {type=Vector2f, description="", defaultValue {struct=Vector2f, value=<0 0>}}
  User.GridSizeWorldSpace {type=Vector2f, description="", defaultValue {struct=Vector2f, value=<0 0>}}
  User.Resolution {type=NiagaraInt32}
  User.Scale {type=NiagaraFloat, defaultValue=1, description=""}
 ]
}
