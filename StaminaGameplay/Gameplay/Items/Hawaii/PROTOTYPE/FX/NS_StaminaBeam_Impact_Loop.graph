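# Looping Niagara impact effect for the prototype Stamina Beam, with continuously spawned explosion puffs and sparks.
# Explo_Large_Center drives mesh puffs through dynamic material curves, alongside CPU sparks split into rising and falling groups and GPU falling sparks.
# A system-level camera-distance value boosts sizes and spark speeds to keep the impact readable afar, while User.Scale sizes the puffs.
# Splash, glow and streak emitters are disabled; LWE_ variants list no renderers.
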NiagaraSystem /StaminaGameplay/Gameplay/Items/Hawaii/PROTOTYPE/FX/NS_StaminaBeam_Impact_Loop
dependencies {
 usedDataInterfaces [
  &NiagaraDataInterfaceColorCurve
  &NiagaraDataInterfaceCurve
  &NiagaraDataInterfaceVector2DCurve
  &NiagaraDataInterfaceCamera
 ]
 usedDynamicInputs [
  &SimulationPosition
  &MakeFloatFromVector2d
  &FN_LOD_Distance_Remap
 ]
 usedModules [
  &SystemState
  &self:SystemSpawnScript.SystemScriptSource.SystemScriptGraph.NiagaraNodeAssignment_0.SetVariables_#1
  &self:SystemSpawnScript.SystemScriptSource.SystemScriptGraph.NiagaraNodeAssignment_1.SetVariables_#2
  &self:SystemSpawnScript.SystemScriptSource.SystemScriptGraph.NiagaraNodeAssignment_3.SetVariables_#3
 ]
}
emitters {
 Explo_Large_Center {
  emitterUpdateScript [
   &EmitterState()
   &SpawnRate(SpawnRate=10, Use Spawn Probability=false)
  ]
  particleSpawnScript [
   &InitializeParticle(
    Lifetime=&UniformRangedFloat(Minimum=0.5, Maximum=0.7)
    Mesh Scale Mode=::NewEnumerator3
    Mesh Uniform Scale=&Add_Float(
     A=&Multiply_Float(A=&UniformRangedFloat(Minimum=1.3, Maximum=1.7), B=$User.Scale)
     B=&Multiply_Float(A=$System.DistanceToCamera, B=1.5)
    )
    UsePositionOffset=false
    Write Material Random=true
   ) {module=InitializeParticle001}
   &DynamicMaterialParameters(
    Index 0 Param 1=&Add_Float(
     A=&UniformRangedFloat(Minimum=1.5, Maximum=1.8)
     B=&Multiply_Float(A=$System.DistanceToCamera, B=4)
    )
    Index 0 Param 3=-0.3
    Module.Param0WriteEnabled=true
    Module.Param1WriteEnabled=false
    Module.Param2WriteEnabled=true
    Module.Param3WriteEnabled=false
   )
   &InitialMeshOrientation(Rotation=&UniformRangedVector(Minimum=<-1 -1 -1>, Maximum=<1 1 1>), Use Rotation Vector=false) {module=InitialMeshOrientation001}
   &ShapeLocation(Handle Radius=1, Large Radius=40, MaskSpawnGroup=false, Override Seed=false, Shape Primitive=::NewEnumerator3, UseOwnerScale=true)
   &AddVelocity(
    Velocity=$Particles.ShapeLocation.ShapeNormal
    Velocity Speed Scale=&RandomRangeFloat(Minimum=150, Maximum=200)
   )
  ]
  particleUpdateScript [
   &ParticleState()
   &Drag(Use Linear Drag=true, Use Rotational Drag=false)
   &SolveForcesAndVelocity(Clamp Velocity=false, Limit Acceleration=false)
   &ScaleColor(
    Scale Alpha=&FloatFromCurve(FloatCurve=DataInterface() {propertyValues {Curve {keys [
         {arriveTangent=-1.8332, interpMode=RCIM_Cubic, time=0, value=0.874}
         {arriveTangent=-0.874, interpMode=RCIM_Cubic, leaveTangent=-0.874, tangentMode=RCTM_Break, time=0.5574, value=0.6351}
         {arriveTangent=-2.2319, interpMode=RCIM_Cubic, leaveTangent=-3.4277, tangentMode=RCTM_User, time=0.9988, value=-0.5}
        ]}}})
    Scale RGB=&MakeVectorFromLinearColorRGB(Color=&ColorFromCurve(ColorCurve=DataInterface() {propertyValues {
        AlphaCurve {keys []}
        BlueCurve {keys [<at 0 0.0026 RCIM_Cubic>,<at 0.2422 0.015>,<at 0.3364 0.0187>,<at 1 0.1091 RCIM_Cubic>]}
        GreenCurve {keys [<at 0 0.0011 RCIM_Cubic>,<at 0.2422 0.0135>,<at 0.3364 0.0171>,<at 1 0.1076 RCIM_Cubic>]}
        RedCurve {keys [<at 0 1.7597 RCIM_Cubic>,<at 0.2422 0.0294>,<at 0.3364 0.0289>,<at 1 0.1115 RCIM_Cubic>]}
       }}))
    ScaleA=true
    ScaleRGB=true
   ) {module=ScaleColor001}
   &DynamicMaterialParameters(
    Index 0 Param 1=&UniformRangedFloat(Minimum=1, Maximum=1.8)
    Index 0 Param 2=&FloatFromCurve(FloatCurve=DataInterface() {propertyValues {Curve {keys [<at 0.13 0.77 RCIM_Cubic>, {interpMode=RCIM_Cubic, leaveTangent=-12.3793, time=1, value=-2}]}}})
    Index 0 Param 3=-0.4
    Index 0 Param 4=&FloatFromCurve(FloatCurve=DataInterface() {propertyValues {Curve {keys [<at 0 2.4 RCIM_Cubic>,<at 1 0.5 RCIM_Cubic>]}}}, Scale Curve=2)
    Module.Param0WriteEnabled=false
    Module.Param1WriteEnabled=true
    Module.Param2WriteEnabled=false
    Module.Param3WriteEnabled=true
   ) {module=DynamicMaterialParameters001}
   &ScaleMeshSize(Scale Factor=&VectorFromFloat(Value=&FloatFromCurve(
      FloatCurve=DataInterface() {propertyValues {Curve {keys [
          {arriveTangent=5.6759, interpMode=RCIM_Cubic, leaveTangent=5.6759, tangentMode=RCTM_User, time=0, value=0}
          {arriveTangent=1, interpMode=RCIM_Cubic, leaveTangent=1, time=0.2218, value=0.6656}
          <at 1 1 RCIM_Cubic>
         ]}}}
      Scale Curve=&Add_Float(A=1.2, B=&Multiply_Float(A=0.3, B=$System.DistanceToCamera))
     )))
   &UpdateMeshOrientation(
    Rotation Rate=&Lerp_Float(A=-1, B=1, Alpha=$Particles.Initial.MaterialRandom)
    Rotation Vector=&Lerp_Vector(A=<-1 0 0>, B=<1 0 1>, Alpha=$Particles.Initial.MaterialRandom)
   ) {module=UpdateMeshOrientation001}
  ]
  rendererClasses [&NiagaraMeshRendererProperties]
  renderers [{rendererClass=&NiagaraMeshRendererProperties, rendererIndex=0}]
 }
 LWE_Sphere_Glow {rendererClasses [], renderers []}
 LWE_Splash {rendererClasses [], renderers []}
 LWE_Streaks {bEnabled=false, rendererClasses [], renderers []}
 LWE_StreaksFat {rendererClasses [], renderers []}
 SparksLazy001 {
  emitterUpdateScript [
   &EmitterState()
   &SpawnRate(SpawnRate=25, Use Spawn Probability=false)
  ]
  particleSpawnScript [
   &FN_SetSpawnGroup(Number of Spawn Groups=2)
   &InitializeParticle(
    AdjustAlpha=false
    AdjustHue=false
    AdjustSaturation=false
    AdjustValue=true
    Color=<rgba 0 0.7453 2 1>
    Color Mode=::NewEnumerator3
    Lifetime Max=1
    Lifetime Min=0.5
    Lifetime Mode=::NewEnumerator1
    Sprite Size Max=<6 20>
    Sprite Size Min=<5 7>
    Sprite Size Mode=::NewEnumerator2
    UsePositionOffset=false
    Value Range:Vector2f=<0.5 5>
    Write Material Random=true
   ) {module=InitializeParticle001}
   /SetVariables(Particles.VisibilityTag=$Particles.SpawnGroup) {isSetParameters=true}
   &SphereLocation(
    MaskSpawnGroup=false
    Override Seed=false
    Sphere Radius=35
    Surface Only=false
    UseNonUniformScale=false
    UseOffset=false
    UseSphereOrientation=false
   ) {module=SphereLocation001}
   &AddVelocity(Velocity=&MaskVectorBySpawnGroup(
     Masked Vector=&RandomRangeVector(Minimum=<250 0 0>, Maximum=<1500 1 1>)
     Passthrough Vector=&RandomRangeVector(Minimum=<50 0 0>, Maximum=<1000 1 1>)
    )) {enabled=false}
   &AddVelocity(Velocity=&Multiply_VectorByFloat(
     Vector=$Output.SphereLocation001.SphereNormal
     Float=&Add_Float(
      A=&RandomRangeFloat(Minimum=250, Maximum=500)
      B=&Multiply_Float(A=$System.DistanceToCamera, B=1000)
     )
    )) {module=AddVelocity001}
   &AddVelocity(
    Cone Angle=75
    Module.UseConeFalloff=false
    Module.UseVelDistribution=false
    Override Seed=false
    Velocity Mode=::NewEnumerator2
    Velocity Speed=&Add_Float(
     A=&RandomRangeFloat(Minimum=700, Maximum=1000)
     B=&Multiply_Float(A=$System.DistanceToCamera, B=1000)
    )
   ) {module=AddVelocity002}
   /SetVariables(Particles.Renderer2_Color:LinearColor=&RandomRangeLinearColor(Minimum=<rgba 0.0191 0.0156 0.0147 1>, Maximum=<rgba 0.1267 0.1035 0.0975 1>)) {isSetParameters=true}
  ]
  particleUpdateScript [
   &ParticleState()
   &AccelerationForce(Acceleration=&MaskVectorBySpawnGroup(Masked Vector=<0 0 700>, Passthrough Vector=<0 0 -900>))
   &Drag(Use Linear Drag=true, Use Rotational Drag=false)
   &SolveForcesAndVelocity(Clamp Velocity=false, Limit Acceleration=false)
   &ScaleSpriteSizeBySpeed(Max Scale Factor=<1 3>, Source Velocity:Vector3f=&Multiply_Vector(A=<0.01 0.01 1>, B=$Particles.Velocity))
   &ScaleColor(
    Scale Alpha=&FloatFromCurve(FloatCurve=DataInterface() {propertyValues {Curve {keys [<at 0 3>,<at 0.2 1>]}}})
    Scale RGB=&MakeVectorFromLinearColorRGB(Color=&ColorFromCurve(ColorCurve=DataInterface() {propertyValues {
        AlphaCurve {keys [<at 0 1 RCIM_Cubic>,<at 1 1 RCIM_Cubic>]}
        BlueCurve {keys [<at 0 0.1858 RCIM_Cubic>,<at 0.4236 0.0295>,<at 1 0 RCIM_Cubic>]}
        GreenCurve {keys [<at 0 0.8852 RCIM_Cubic>,<at 0.4236 0.1515>,<at 1 0.0031 RCIM_Cubic>]}
        RedCurve {keys [<at 0 1 RCIM_Cubic>,<at 0.4236 1>,<at 1 1 RCIM_Cubic>]}
       }}))
    ScaleA=true
    ScaleRGB=false
   )
   &ScaleSpriteSize(Scale Factor=&Vector2DFromFloat(Value=&FloatFromCurve(FloatCurve=DataInterface() {propertyValues {Curve {keys [
          {tangentMode=RCTM_User, time=0, value=1}
          {arriveTangent=-0.0405, interpMode=RCIM_Cubic, leaveTangent=-0.0405, tangentMode=RCTM_User, time=0.7009, value=1}
          {arriveTangent=-8.0089, interpMode=RCIM_Cubic, leaveTangent=-8.0089, tangentMode=RCTM_User, time=1, value=0}
         ]}}})))
   /SetVariables(Particles.SpriteSize=/Niagara/DynamicInputs/Add/Add_Vector2(
     A:Vector2f=$Particles.SpriteSize
     B:Vector2f=&Vector2DFromFloat(Value=&Multiply_Float(A=$System.DistanceToCamera, B=75))
    )) {isSetParameters=true}
   /SetVariables(Particles.Rednerer2_Size:Vector2f=&Vector2DFromFloat(Value=&FloatFromCurve(
      FloatCurve=DataInterface() {propertyValues {Curve {keys [
          {tangentMode=RCTM_User, time=0, value=1}
          {arriveTangent=-0.0405, interpMode=RCIM_Cubic, leaveTangent=-0.0405, tangentMode=RCTM_User, time=0.7009, value=1}
          {arriveTangent=-8.0089, interpMode=RCIM_Cubic, leaveTangent=-8.0089, tangentMode=RCTM_User, time=1, value=0}
         ]}}}
      Scale Curve=&Add_Float(
       A=&RandomRangeFloat(Minimum=7, Maximum=12)
       B=&Multiply_Float(A=$System.DistanceToCamera, B=100)
      )
     ))) {isSetParameters=true}
  ]
  rendererClasses [&NiagaraSpriteRendererProperties]
  renderers [
   {rendererClass=&NiagaraSpriteRendererProperties}
   {rendererClass=&NiagaraSpriteRendererProperties, rendererIndex=1}
  ]
 }
 SparksLazy002 {
  simTarget=GPUComputeSim
  emitterUpdateScript [
   &EmitterState()
   &SpawnRate(SpawnRate=50, Use Spawn Probability=false)
  ]
  particleSpawnScript [
   &InitializeParticle(
    AdjustAlpha=false
    AdjustHue=false
    AdjustSaturation=false
    AdjustValue=true
    Color=<rgba 0 1.1179 3 1>
    Color Mode=::NewEnumerator3
    Lifetime Max=1.5
    Lifetime Min=0.8
    Lifetime Mode=::NewEnumerator1
    Sprite Size Max=<9 9>
    Sprite Size Min=<3.5 3.5>
    Sprite Size Mode=::NewEnumerator2
    UsePositionOffset=false
    Value Range:Vector2f=<0.5 5>
    Write Material Random=true
   ) {module=InitializeParticle001}
   &SphereLocation(
    MaskSpawnGroup=false
    Override Seed=false
    Sphere Radius=25
    Surface Only=false
    UseNonUniformScale=false
    UseOffset=false
    UseSphereOrientation=false
   ) {module=SphereLocation001}
   &AddVelocity(Velocity=&RandomRangeVector(Minimum=<50 0 0>, Maximum=<1000 0 0>, Evaluation Type=::NewEnumerator1)) {enabled=false}
   &AddVelocity(Velocity=&Multiply_VectorByFloat(
     Vector=$Output.SphereLocation001.SphereNormal
     Float=&Add_Float(
      A=&RandomRangeFloat(Minimum=250, Maximum=500)
      B=&Multiply_Float(A=$System.DistanceToCamera, B=1000)
     )
    )) {module=AddVelocity001}
   &AddVelocity(
    Cone Angle=75
    Module.UseConeFalloff=false
    Module.UseVelDistribution=false
    Override Seed=false
    Velocity Mode=::NewEnumerator2
    Velocity Speed=&Add_Float(
     A=&RandomRangeFloat(Minimum=750, Maximum=1100)
     B=&Multiply_Float(A=$System.DistanceToCamera, B=1000)
    )
   ) {module=AddVelocity002}
  ]
  particleUpdateScript [
   &ParticleState()
   &AccelerationForce(Acceleration=<0 0 -900>)
   &Drag(Drag=2, Use Linear Drag=true, Use Rotational Drag=false)
   &Collision(
    Age Colliding Particles=false
    Kill Offscreen Particles=true
    Module.Control Rest State Via Surface Normal=false
    Module.EnableMaxCollisionCount=false
    Module.Randomize Collision Normal Vector=false
    Restitution=0.1
   ) {enabled=false}
   &SolveForcesAndVelocity(Clamp Velocity=false, Limit Acceleration=false)
   &ScaleSpriteSizeBySpeed(Max Scale Factor=<1 3>, Source Velocity:Vector3f=&Multiply_Vector(A=<0.01 0.01 1>, B=$Particles.Velocity))
   &ScaleColor(
    Scale Alpha=&FloatFromCurve(FloatCurve=DataInterface() {propertyValues {Curve {keys [<at 0 3>,<at 0.2 1>]}}})
    Scale RGB=&MakeVectorFromLinearColorRGB(Color=&ColorFromCurve(ColorCurve=DataInterface() {propertyValues {
        AlphaCurve {keys [<at 0 1 RCIM_Cubic>,<at 1 1 RCIM_Cubic>]}
        BlueCurve {keys [<at 0.2395 2>, {arriveTangent=-1.3477, interpMode=RCIM_Cubic, leaveTangent=-1.3477, time=0.6646, value=2}, <at 0.9815 1>]}
        GreenCurve {keys [<at 0.2395 2>, {arriveTangent=-1.3477, interpMode=RCIM_Cubic, leaveTangent=-1.3477, time=0.6646, value=2}, <at 0.9815 1>]}
        RedCurve {keys [<at 0.2395 2>, {arriveTangent=-1.3477, interpMode=RCIM_Cubic, leaveTangent=-1.3477, time=0.6646, value=2}, <at 0.9815 1>]}
       }}))
    ScaleA=true
    ScaleRGB=true
   )
   &ScaleSpriteSize(Scale Factor=&Vector2DFromFloat(Value=&FloatFromCurve(FloatCurve=DataInterface() {propertyValues {Curve {keys [
          {tangentMode=RCTM_User, time=0, value=1}
          {arriveTangent=-0.0405, interpMode=RCIM_Cubic, leaveTangent=-0.0405, tangentMode=RCTM_User, time=0.7009, value=1}
          {arriveTangent=-8.0089, interpMode=RCIM_Cubic, leaveTangent=-8.0089, tangentMode=RCTM_User, time=1, value=0}
         ]}}})))
   /SetVariables(Particles.SpriteSize=&Add_Vector2(
     A:Vector2f=$Particles.SpriteSize
     B:Vector2f=&Vector2DFromFloat(Value=&Multiply_Float(A=$System.DistanceToCamera, B=55))
    )) {isSetParameters=true}
  ]
  rendererClasses [&NiagaraSpriteRendererProperties]
  renderers [{rendererClass=&NiagaraSpriteRendererProperties}]
 }
 SphereGlow {
  bEnabled=false
  emitterUpdateScript [
   &EmitterState(Life Cycle Mode=::NewEnumerator1, Loop Behavior=::NewEnumerator1, Loop Duration=7, UseLoopDelay=false)
   &SpawnBurst_Instantaneous(Spawn Count=1, Use Spawn Probability=false)
  ]
  particleSpawnScript [&InitializeParticle(
    Color=<rgba 0 0.7453 2 1>
    Mesh Scale Mode=::NewEnumerator3
    Mesh Uniform Scale=&MakeFloatFromVector(
     Vector=&VectorFromFloat(Value=&Multiply_Float(A=27, B=$User.Scale))
     Channel=::NewEnumerator0
    )
    Sprite Size Mode=::NewEnumerator3
    Uniform Sprite Size=250
    UsePositionOffset=false
    Write Material Random=true
   ) {module=InitializeParticle001}]
  particleUpdateScript [
   &ParticleState(Kill Particles When Lifetime Has Elapsed=false, Let Infinitely Lived Particles Die When Emitter Deactivates=true)
   /SetVariables(Particles.DistanceToCamera:NiagaraFloat=&RemapRange(
     Input Value To Remap=&VectorLength(Vector=&Subtract_Vector(
       A=Linked() {linkedVariable {name=Engine.Owner.Position, type=NiagaraPosition}}
       B=&ReturnCameraVector(Camera Property=::NewEnumerator0, Coordinate Space=::Simulation)
      ))
     Input Min=5000
     Input Max=12000
     Output Min=0
     Output Max=1
    )) {enabled=false, isSetParameters=true}
   &ScaleMeshSize(Scale Factor=&VectorFromFloat(Value=&FloatFromCurve(
      FloatCurve=DataInterface() {propertyValues {Curve {keys [
          {arriveTangent=9.6835, interpMode=RCIM_Cubic, leaveTangent=9.6835, tangentMode=RCTM_User, time=0, value=0}
          {arriveTangent=0.025, interpMode=RCIM_Cubic, leaveTangent=0.025, tangentMode=RCTM_User, time=0.1, value=1}
         ]}}}
      Scale Curve=&Add_Float(A=1, B=&Multiply_Float(A=2.5, B=$System.DistanceToCamera))
     )))
  ]
  rendererClasses [&NiagaraMeshRendererProperties]
  renderers [{rendererClass=&NiagaraMeshRendererProperties, rendererIndex=0}]
 }
 Splash {
  bEnabled=false
  emitterUpdateScript [
   &EmitterState()
   &SpawnRate(SpawnRate=6, Use Spawn Probability=false)
  ]
  particleSpawnScript [
   &InitializeParticle(
    Color=<rgba 0 1.8631 5 1>
    Lifetime=0.3
    Mesh Scale Mode=::NewEnumerator3
    Mesh Uniform Scale=&Add_Float(
     A=&Multiply_Float(A=2.25, B=$User.Scale)
     B=&Multiply_Float(A=$System.DistanceToCamera, B=1)
    )
    UsePositionOffset=false
    Write Material Random=true
   ) {module=InitializeParticle001}
   &InitialMeshOrientation(Mesh Orientation Mode=::NewEnumerator2, Orientation Vector=<0 0 1>, Use Rotation Vector=false) {module=InitialMeshOrientation001}
   &DynamicMaterialParameters(
    Index 0 Param 1=&UniformRangedFloat(Minimum=0, Maximum=20)
    Module.Param0WriteEnabled=true
    Module.Param1WriteEnabled=false
    Module.Param2WriteEnabled=false
    Module.Param3WriteEnabled=false
   )
  ]
  particleUpdateScript [
   &ParticleState()
   &SolveForcesAndVelocity(Clamp Velocity=false, Limit Acceleration=false)
   &ScaleMeshSize(Scale Factor=&VectorFromFloat(Value=&FloatFromCurve(FloatCurve=DataInterface() {propertyValues {Curve {keys [{arriveTangent=9.901, interpMode=RCIM_Cubic, leaveTangent=9.901, tangentMode=RCTM_User, time=0, value=0.4}, <at 1 6 RCIM_Cubic>]}}})))
   &ScaleColor(
    Scale Alpha=&FloatFromCurve(FloatCurve=DataInterface() {propertyValues {Curve {keys [<at 0.4 1 RCIM_Cubic>,<at 1 0>]}}})
    ScaleA=true
    ScaleRGB=true
   )
  ]
  rendererClasses [&NiagaraMeshRendererProperties]
  renderers [{rendererClass=&NiagaraMeshRendererProperties, rendererIndex=0}]
 }
 Streaks001 {
  bEnabled=false
  emitterUpdateScript [
   &EmitterState()
   &SpawnBurst_Instantaneous(Spawn Count=1, Use Spawn Probability=false)
   &SpawnRate(SpawnRate=6, Use Spawn Probability=false)
  ]
  particleSpawnScript [
   &InitializeParticle(
    Lifetime=0.45
    Mesh Scale Mode=::NewEnumerator3
    Mesh Uniform Scale=&Add_Float(
     A=&RandomRangeFloat(Minimum=0.1, Maximum=0.2)
     B=&Multiply_Float(A=$System.DistanceToCamera, B=1)
    )
    UsePositionOffset=false
    Write Material Random=true
   ) {module=InitializeParticle001}
   &InitialMeshOrientation(Use Rotation Vector=false)
  ]
  particleUpdateScript [
   &ParticleState()
   &ScaleMeshSize(Scale Factor=&VectorFromFloat(Value=&FloatFromCurve(FloatCurve=DataInterface() {propertyValues {Curve {keys [{arriveTangent=2.9152, interpMode=RCIM_Cubic, leaveTangent=2.9152, tangentMode=RCTM_User, time=0, value=0}, <at 1 1 RCIM_Cubic>]}}})))
   &DynamicMaterialParameters(
    Index 0 Param 2=1
    Index 0 Param 3=&FloatFromCurve(FloatCurve=DataInterface() {propertyValues {Curve {keys [{arriveTangent=-25.4246, time=0, value=0}, <at 1 1>]}}})
    Index 0 Param 4=&FloatFromCurve(
     FloatCurve=DataInterface() {propertyValues {Curve {keys [<at 0 0.5>,<at 1 1>]}}}
     Scale Curve=&Add_Float(A=1, B=&Multiply_Float(A=$System.DistanceToCamera, B=10))
    )
    Module.Param0WriteEnabled=true
    Module.Param1WriteEnabled=true
    Module.Param2WriteEnabled=true
    Module.Param3WriteEnabled=true
   )
   &ScaleColor(
    Scale Alpha=&FloatFromCurve(FloatCurve=DataInterface() {propertyValues {Curve {keys [<at 0 1>,<at 1 0>]}}})
    ScaleA=true
    ScaleRGB=true
   )
  ]
  rendererClasses [&NiagaraMeshRendererProperties]
  renderers [{rendererClass=&NiagaraMeshRendererProperties, rendererIndex=0}]
 }
 Streaks_Fat {
  bEnabled=false
  emitterUpdateScript [
   &EmitterState()
   &SpawnRate(SpawnRate=20, Use Spawn Probability=false) {module=SpawnRate001}
  ]
  particleSpawnScript [
   &InitializeParticle(
    Color=<rgba 0 1.8631 5 1>
    Lifetime=&UniformRangedFloat(Minimum=0.35, Maximum=0.5)
    Sprite Size=&Multiply_Vector2D_ByFloat(Float=$User.Scale, Vector2D=&UniformRangedVector2D(Minimum=<200 500>, Maximum=<250 700>))
    Sprite Size Mode=::NewEnumerator4
    UsePositionOffset=false
    Write Material Random=true
   ) {module=InitializeParticle001}
   /Niagara/Modules/Spawn/Location/SphereLocation(MaskSpawnGroup=false, Module.Override Randomness=false, Sphere Radius=12, Surface Only=true)
   &AddVelocityFromPoint(
    Module.Use Falloff Distance=false
    Velocity Origin:Vector3f=<0 0 0>
    Velocity Strength=&UniformRangedFloat(Minimum=100, Maximum=200)
   )
  ]
  particleUpdateScript [
   &ParticleState()
   &SolveForcesAndVelocity(Clamp Velocity=false, Limit Acceleration=false)
   &ScaleSpriteSize(Scale Factor=&Vector2DFromCurve(
     Vector2Curve=DataInterface() {propertyValues {XCurve {keys [<at 0 0.75 RCIM_Cubic>,<at 1 1 RCIM_Cubic>]}, YCurve {keys [<at 0 0 RCIM_Cubic>,<at 1 2 RCIM_Cubic>]}}}
     ScaleCurve=<1 1>
    ))
   &ScaleColor(
    Scale Alpha=&FloatFromCurve(FloatCurve=DataInterface() {propertyValues {Curve {keys [<at 0 1 RCIM_Cubic>,<at 1 0>]}}})
    Scale RGB=&VectorFromFloat(Value=&FloatFromCurve(FloatCurve=DataInterface() {propertyValues {Curve {keys [<at 0 10 RCIM_Cubic>,<at 1 1 RCIM_Cubic>]}}}))
    ScaleA=true
    ScaleRGB=true
   )
  ]
  rendererClasses [&NiagaraSpriteRendererProperties]
  renderers [{rendererClass=&NiagaraSpriteRendererProperties}]
 }
}
summary {
 systemName=NS_StaminaBeam_Impact_Loop
 emitters [
  {emitterName=Explo_Large_Center, rendererClasses [&NiagaraMeshRendererProperties]}
  {bEnabled=false, emitterName=Splash, rendererClasses [&NiagaraMeshRendererProperties]}
  {bEnabled=false, emitterName=Streaks_Fat, rendererClasses [&NiagaraSpriteRendererProperties]}
  {bEnabled=false, emitterName=SphereGlow, rendererClasses [&NiagaraMeshRendererProperties]}
  {emitterName=SparksLazy001, rendererClasses [&NiagaraSpriteRendererProperties]}
  {emitterName=SparksLazy002, simTarget=GPUComputeSim, rendererClasses [&NiagaraSpriteRendererProperties]}
  {bEnabled=false, emitterName=Streaks001, rendererClasses [&NiagaraMeshRendererProperties]}
  {emitterName=LWE_Sphere_Glow, rendererClasses []}
  {emitterName=LWE_Splash, rendererClasses []}
  {emitterName=LWE_StreaksFat, rendererClasses []}
  {bEnabled=false, emitterName=LWE_Streaks, rendererClasses []}
 ]
 userVariables [
  User.bIsFrontEndPreview {type=NiagaraBool}
  User.PlayerSpeed {type=NiagaraFloat, defaultValue=0, description=""}
  User.Scale {type=NiagaraFloat, defaultValue=0.2, description=""}
 ]
}
