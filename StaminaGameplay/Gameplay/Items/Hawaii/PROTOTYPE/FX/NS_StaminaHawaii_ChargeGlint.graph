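# Niagara system for the prototype StaminaHawaii item's charge glint, a short blue sprite flare attached to a skeletal mesh.
# A burst emitter spawns the glint, whose position blends by a timeline value from an offset off one mesh sample toward the midpoint of two samples.
# Size and brightness rise and fall over lifetime, with size boosted by camera distance while the sprite spins.
# Module-written color and size variants supply layers for four sprite renderers.
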NiagaraSystem /StaminaGameplay/Gameplay/Items/Hawaii/PROTOTYPE/FX/NS_StaminaHawaii_ChargeGlint
dependencies {
 usedDataInterfaces [
  &NiagaraDataInterfaceCamera
  &NiagaraDataInterfaceCurve
  &NiagaraDataInterfaceSkeletalMesh
  &NiagaraDataInterfaceArrayFloat
 ]
 usedDynamicInputs [&SetBoolByIntComparison, &FN_LOD_Distance_Remap]
 usedModules [
  &SystemState
  &self:SystemSpawnScript.SystemScriptSource.SystemScriptGraph.NiagaraNodeAssignment_1.SetVariables_#1
  &SampleSkeletalMesh
  &Timeline
  &self:SystemSpawnScript.SystemScriptSource.SystemScriptGraph.NiagaraNodeAssignment_0.SetVariables_#2
  &self:SystemSpawnScript.SystemScriptSource.SystemScriptGraph.NiagaraNodeAssignment_2.SetVariables_#3
 ]
}
emitters {Empty {
  emitterUpdateScript [
   &EmitterState()
   &SpawnBurst_Instantaneous(Use Spawn Probability=false, UseLoopCountLimit=false)
  ]
  particleSpawnScript [
   &InitializeParticle(
    Color=<rgba 0 1.1179 3 1>
    Lifetime=2
    Sprite Size Mode=::NewEnumerator3
    Uniform Sprite Size=500
    UsePositionOffset=false
    Write Material Random=true
   )
   /SetVariables(Particles.DistanceToCamera:NiagaraFloat=&NormalizeDistanceRange(
     Start Position:NiagaraPosition=$Particles.Position
     End Position:NiagaraPosition=&ReturnCameraVector(Camera Property=::NewEnumerator0, Coordinate Space=::World)
     Distance:NiagaraFloat=20000
     Invert Normalized Range:NiagaraBool=false
    )) {enabled=false, isSetParameters=true}
  ]
  particleUpdateScript [
   &ParticleState()
   /SetVariables(Particles.Position=/Niagara/DynamicInputs/Lerp/Lerp_Position(
     A:NiagaraPosition=&Add_Vector(
      A=$System.SampleSkeletalMesh.SampledPosition
      B=&RotateVectorByQuaternion(Vector To Rotate=<-10 7 0>, Quaternion=$System.SampleSkeletalMesh.SampledRotation)
     )
     B:NiagaraPosition=&Lerp_Position(
      A:NiagaraPosition=Linked() {linkedVariable {name=System.SampleSkeletalMesh.SampledPosition, type=Vector3f}}
      B:NiagaraPosition=$System.SampleSkeletalMesh001.SampledPosition
      Alpha:NiagaraFloat=0.5
     )
     Alpha:NiagaraFloat=$System.Timeline.Float
    )) {isSetParameters=true}
   /Niagara/Modules/Update/Camera/MaintainInCameraParticleScale(Half FOV=25, Render Target Width=0.5) {enabled=false}
   &ScaleSpriteSize(
    Uniform Curve Scale=&Multiply_Float(A=$System.DistanceToCamera, B=50)
    Uniform Curve Sprite Scale=DataInterface() {propertyValues {Curve {keys [
        {interpMode=RCIM_Cubic, tangentMode=RCTM_User, time=0, value=0}
        {interpMode=RCIM_Cubic, tangentMode=RCTM_User, time=0.5, value=1}
        {tangentMode=RCTM_User, time=1, value=0}
       ]}}}
   ) {enabled=false}
   &ScaleSpriteSize(
    Scale Sprite Size Mode=::NewEnumerator0
    Uniform Scale Factor=&FloatFromCurve(
     FloatCurve=DataInterface() {propertyValues {Curve {
        defaultValue=1
        keys [
         {interpMode=RCIM_Cubic, tangentMode=RCTM_User, time=0, value=0}
         {interpMode=RCIM_Cubic, tangentMode=RCTM_User, time=0.5, value=1}
         {tangentMode=RCTM_User, time=1, value=0}
        ]
       }}}
     Scale Curve=&Add_Float(A=1, B=&Multiply_Float(A=$System.DistanceToCamera, B=4))
    )
   ) {module=ScaleSpriteSize001}
   &SpriteRotationRate(Rotation Rate=100)
   &ScaleColor(
    Scale RGB=&VectorFromFloat(Value=&FloatFromCurve(FloatCurve=DataInterface() {propertyValues {Curve {keys [
          {interpMode=RCIM_Cubic, tangentMode=RCTM_User, time=0, value=0.5}
          {interpMode=RCIM_Cubic, tangentMode=RCTM_User, time=0.5, value=1.5}
          {tangentMode=RCTM_User, time=1, value=0.5}
         ]}}}))
    ScaleA=true
    ScaleRGB=true
   )
   &FN_Sprite_Attributes(
    Color=<rgba 0.3438 1.3335 3 1>
    Create Color Attribute=true
    Create Sprite Rotation Attribute=true
    Create SpriteSize Attribute=true
    SpriteSize=&Multiply_Vector2D_ByFloat(Float=1.5, Vector2D=$Particles.SpriteSize)
    Write to Module Namespace=true
   )
   &FN_Sprite_Attributes(
    Color=<rgba 0 0.6345 2 1>
    Create Color Attribute=true
    Create SpriteSize Attribute=true
    SpriteSize=&Multiply_Vector2D_ByFloat(Float=0.8, Vector2D=$Particles.SpriteSize)
    Write to Module Namespace=true
   ) {module=FN_Sprite_Attributes001}
   &FN_Sprite_Attributes(
    Color=<rgba 0 0.008 0.0191 1>
    Create Color Attribute=true
    Create SpriteSize Attribute=true
    SpriteSize=&Multiply_Vector2D_ByFloat(Float=1.7, Vector2D=$Particles.SpriteSize)
    Write to Module Namespace=true
   ) {module=FN_Sprite_Attributes002}
  ]
  rendererClasses [&NiagaraSpriteRendererProperties]
  renderers [
   {rendererClass=&NiagaraSpriteRendererProperties}
   {rendererClass=&NiagaraSpriteRendererProperties, rendererIndex=1}
   {rendererClass=&NiagaraSpriteRendererProperties, rendererIndex=2}
   {rendererClass=&NiagaraSpriteRendererProperties, rendererIndex=3}
  ]
 }}
summary {
 systemName=NS_StaminaHawaii_ChargeGlint
 emitters [{emitterName=Empty, rendererClasses [&NiagaraSpriteRendererProperties]}]
 userVariables [User.EquipState {type=NiagaraInt32}]
}
