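# Prototype Niagara camera lens effect for the victim of a Hawaii item, showing a directional splash.
# A single emitter bursts one short-lived sprite, using User.Direction to pick its offset (up, down or either side) and size from arrays.
# FN_Sprite_Attributes writes an extra dark color and doubled size in its module namespace, alongside the bright blue base particle.
# Two sprite renderers draw the particle, with randomized material values.
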NiagaraSystem /StaminaGameplay/Gameplay/Items/Hawaii/PROTOTYPE/FX/NS_Victim_CameraLensFX
dependencies {
 usedDataInterfaces [&NiagaraDataInterfaceArrayFloat3, &NiagaraDataInterfaceArrayFloat2]
 usedDynamicInputs []
 usedModules [&SystemState]
}
emitters {SplashRadial {
  emitterUpdateScript [
   &EmitterState()
   &SpawnBurst_Instantaneous(Spawn Count=1, Use Spawn Probability=false)
  ]
  particleSpawnScript [&InitializeParticle(
    Color=<rgba 0.7923 2.8996 5 0.5>
    Lifetime=2
    Position Offset=&SelectVectorFromArray(
     Direct Array Index=$User.Direction
     Vector Selection Array=DataInterface() {propertyValues {FloatData [<0 0 65>, <0 0 -65>, <0 -100 0>, <0 100 0>]}}
    )
    Sprite Size=&SelectVector2DFromArray(
     Direct Array Index=$User.Direction
     Vector 2D Selection Array=DataInterface() {propertyValues {FloatData [<200 80>, <200 80>, <80 80>, <80 80>]}}
    )
    Sprite Size Mode=::NewEnumerator4
    UsePositionOffset=true
    Write Material Random=true
   )]
  particleUpdateScript [
   &FN_Sprite_Attributes(
    Color=<rgba 0.0365 0 0.0001 1>
    Create Color Attribute=true
    Create SpriteSize Attribute=true
    SpriteSize=&Multiply_Vector2D_ByFloat(Float=2, Vector2D=$Particles.SpriteSize)
    Write to Module Namespace=true
   )
   &ParticleState()
   /SetVariables() {isSetParameters=true}
  ]
  rendererClasses [&NiagaraSpriteRendererProperties]
  renderers [
   {rendererClass=&NiagaraSpriteRendererProperties}
   {rendererClass=&NiagaraSpriteRendererProperties, rendererIndex=1}
  ]
 }}
summary {
 systemName=NS_Victim_CameraLensFX
 emitters [{emitterName=SplashRadial, rendererClasses [&NiagaraSpriteRendererProperties]}]
 userVariables [User.Direction {type=NiagaraInt32}]
}
